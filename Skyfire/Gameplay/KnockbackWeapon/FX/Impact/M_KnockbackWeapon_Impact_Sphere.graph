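# Translucent unlit mesh-particle material for the knockback weapon's impact sphere effect.
# Adjustable Fresnel is optionally multiplied by two differently tiled fire-noise samples, panned vertically at a dynamic-parameter speed.
# The result scales particle alpha into a depth fade, while particle color feeds emissive adjustments.
# The inverted-Fresnel blend receives identical inputs, and extra Fresnel-band and refraction branches are only partly wired.
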Material /Skyfire/Gameplay/KnockbackWeapon/FX/Impact/M_KnockbackWeapon_Impact_Sphere
graph {
 properties {
  bUsedWithMeshParticles=true
  bUsedWithNiagaraMeshParticles=true
  blendMode=BLEND_Translucent
  pixelDepthOffsetMode=PDOM_Legacy
  shadingModel=MSM_Unlit
  usedShadingModels=Unlit
 }
 nodes [
  n1 = MaterialFunctionCall(
   Brightness=ScalarParameter() {defaultValue=1, group=Fresnel, parameterName=Fresnel Brightness, sortPriority=0}
   Exponent=ScalarParameter() {defaultValue=1, group=Fresnel, parameterName=Fresnel Exponent, sortPriority=1}
  ) {materialFunction=&MF_Fresnel_FX}
  n2 = ScalarParameter() {group=Fresnel, parameterName=Use Inverted Fresnel, sortPriority=3}
  n3 = Multiply(A=Time(), B=DynamicParameter() {defaultValue=<rgba 3 1 1 1>, paramNames [Noise Pan Speed, "", "", ""]}.Noise Pan Speed)
  n4 = TextureSample(UVs=Panner(Coordinate=TextureCoordinate(), Time=@n3) {speedY=-1}) {texture=/Game/Global/Textures/Default/Noise/T_FireNoise_02}
  n5 = TextureSample(UVs=Panner(Coordinate=TextureCoordinate() {uTiling=4, vTiling=2}, Time=@n3) {speedY=-1}) {texture=&T_FireNoise_02}
  n6 = Multiply(
   A=Power(Base=Multiply(A=@n4.G, B=@n5.G), Exp=ScalarParameter() {defaultValue=1, group=Noise, parameterName=Noise Power, sortPriority=2})
   B=ScalarParameter() {defaultValue=1, group=Noise, parameterName=Noise Multiplier, sortPriority=1}
  )
  n7 = Multiply(
   A=Multiply(
    A=LinearInterpolate(A=@n1.Fresnel, B=@n1.Fresnel, Alpha=@n2)
    B=ScalarParameter() {defaultValue=1, group=Fresnel, parameterName=Fresnel Intensity, sortPriority=2}
   )
   B=LinearInterpolate(A=Constant() {r=1}, B=Saturate(None=@n6), Alpha=ScalarParameter() {group=Noise, parameterName=Use Noise, sortPriority=0})
  )
  DepthFade(
   Opacity=Multiply(A=ParticleColor().A, B=Clamp(None=@n7))
   FadeDistance=ScalarParameter() {defaultValue=100, group=Opacity, parameterName=DepthFadeDIst, sortPriority=0}
  )
  MaterialFunctionCall(Emissive=ParticleColor().RGB) {materialFunction=&MF_FX_EmissiveAdjustments}
  n8 = DynamicParameter() {defaultValue=<rgba 3 1 1 1>, paramNames [Noise Pan Speed, "", "", ""]}
  n9 = Fresnel(ExponentIn=@n8.G) {baseReflectFraction=0.001}
  n10 = Fresnel(ExponentIn=@n8.Noise Pan Speed) {baseReflectFraction=0.001}
  Multiply(
   A=StaticSwitchParameter(True=OneMinus(None=@n9), False=OneMinus(None=@n10)) {desc=Allows independent control of middle and edge fresnel fade, parameterName=Independent Fresnel Fade}
   B=@n10
  )
  n11 = Fresnel(ExponentIn=Constant() {r=0.8}) {baseReflectFraction=0.001}
  n12 = OneMinus(None=@n11)
  Power(Base=Multiply(A=Multiply(A=@n12, B=@n11)) {constA=1})
  n13 = MaterialFunctionCall("FadeSensitivity (optional)"=Constant() {r=2}) {materialFunction=&FadeBasedOnViewAngle}
  StaticSwitchParameter(True=LinearInterpolate(Alpha=Multiply(B=@n13.Result)) {constA=1, constB=1.1}, False=Constant() {r=1}) {parameterName=UseRefraction}
 ]
}
