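# Translucent, unlit, two-sided material for Niagara mesh particles, named for the noisy rings in the knockback weapon's impact effect.
# It samples an atmospheric cloud noise texture on UVs multiplied by a constant two-axis tiling vector.
# The noise is added to an MF_Fresnel_FX rim term, saturated, and fed into the shared MF_FX_EmissiveAdjustments function.
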Material /Skyfire/Gameplay/KnockbackWeapon/FX/Impact/M_Knockback_NoisyRings
graph {
 properties {
  bUsedWithNiagaraMeshParticles=true
  blendMode=BLEND_Translucent
  pixelDepthOffsetMode=PDOM_Legacy
  shadingModel=MSM_Unlit
  twoSided=true
  usedShadingModels=Unlit
 }
 nodes [
  Constant() {r=1}
  Constant() {r=1}
  n1 = Constant() {r=1}
  n2 = Constant() {r=1}
  n3 = TextureSample(UVs=Multiply(A=TextureCoordinate(), B=AppendVector(A=@n1, B=@n2))) {texture=&AtmosphericCloudNoise01}
  n4 = MaterialFunctionCall(Brightness=Constant() {r=1}, Exponent=Constant() {r=3}) {materialFunction=&MF_Fresnel_FX}
  MaterialFunctionCall(Emissive=Saturate(None=Add(A=@n3.RGB, B=@n4.Fresnel))) {materialFunction=&MF_FX_EmissiveAdjustments}
 ]
}
