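# Niagara impact burst for the knockback weapon, layering short-lived glow, distortion, smoke and noise-sphere effects.
# Active emitters shown include a glow sprite with light attributes, an expanding distortion mesh, smoke thrown outward from a torus, and an inner noise sphere.
# User parameters drive colors, impact size, delay and lifetime, while time-of-day selectors tint smoke and sphere.
# Several emitters, including impact rings, ground ribbons and extra glows, are disabled.
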NiagaraSystem /Skyfire/Gameplay/KnockbackWeapon/FX/Impact/NS_KnockbackWeapon_Impact
dependencies {
 usedDataInterfaces [
  &NiagaraDataInterfaceCurve
  &NiagaraDataInterfaceVector2DCurve
  &NiagaraDataInterfaceColorCurve
  &NiagaraDataInterfaceVectorCurve
  &NiagaraDataInterfaceParticleRead
 ]
 usedDynamicInputs []
 usedModules [&SystemState]
}
emitters {
 AirGlow {
  emitterUpdateScript [
   &EmitterState()
   &SpawnBurst_Instantaneous(Spawn Count=1, Use Spawn Probability=false)
  ]
  particleSpawnScript [/Niagara/Modules/Spawn/Initialization/InitializeParticle(
    Color=$User.Color_01
    Lifetime=&RandomRangeFloat(Minimum=0.23, Maximum=0.25)
    Module.Write Color=true
    Module.Write Lifetime=true
    Module.Write Mass=true
    Module.Write Material Random=true
    Module.Write Position=false
    Module.Write Scale=false
    Module.Write SpriteRotation=false
    Module.Write SpriteSize=true
    Sprite Size=&Vector2DFromFloat(Value=&Multiply_Float(A=&RandomRangeFloat(Minimum=700, Maximum=750, Evaluation Type=::NewEnumerator1), B=$User.ImpactSize))
   )]
  particleUpdateScript [
   &ParticleState()
   &ScaleSpriteSize(Scale Factor=&Vector2DFromFloat(Value=&FloatFromCurve(
      FloatCurve=DataInterface() {propertyValues {Curve {keys [
          {arriveTangent=2.6083, interpMode=RCIM_Cubic, leaveTangent=0.4318, tangentMode=RCTM_User, time=0, value=0.5}
          {arriveTangent=1.3014, interpMode=RCIM_Cubic, leaveTangent=1.3014, tangentMode=RCTM_User, time=0.3, value=0.8}
          {arriveTangent=0.1331, leaveTangent=0.1331, time=1, value=1.2}
         ]}}}
      Scale Curve=1.25
     )))
   &Color(Color=&ColorFromCurve(ColorCurve=DataInterface() {propertyValues {
       AlphaCurve {keys [<at 0 1 RCIM_Cubic>,<at 1 1 RCIM_Cubic>]}
       BlueCurve {keys [<at 0.3144 0.9911 RCIM_Cubic>,<at 1 0.7741 RCIM_Cubic>]}
       GreenCurve {keys [<at 0.3144 0.6197 RCIM_Cubic>,<at 1 0.4 RCIM_Cubic>]}
       RedCurve {keys [<at 0.3144 0.242 RCIM_Cubic>,<at 1 1 RCIM_Cubic>]}
      }})) {enabled=false}
   &ScaleColor(
    Color Value To Scale=$Particles.Color
    Scale Alpha=&FloatFromCurve(FloatCurve=DataInterface() {propertyValues {Curve {keys [<at 0 1 RCIM_Cubic>,<at 1 0 RCIM_Cubic>]}}}, Scale Curve=0.1)
    Scale RGB=&VectorFromFloat(Value=&FloatFromCurve(
      FloatCurve=DataInterface() {propertyValues {Curve {keys [{interpMode=RCIM_Cubic, tangentMode=RCTM_User, time=0, value=1}, {interpMode=RCIM_Cubic, tangentMode=RCTM_User, time=1, value=0}]}}}
      Scale Curve=10
     ))
    ScaleA=true
    ScaleRGB=true
   )
   &CameraOffset(Camera Offset Amount=30)
   &DynamicMaterialParameters(
    Index 0 Param 1=1
    Index 0 Param 2=1
    Index 0 Param 3=1
    Index 0 Param 4=1
    Module.Param0WriteEnabled=true
    Module.Param1WriteEnabled=true
    Module.Param2WriteEnabled=true
    Module.Param3WriteEnabled=true
   )
   &FN_Light_Attributes(
    Create Color Attribute=true
    Create Exponent Attribute=true
    Create Radius Attribute=true
    Enable Particle Light Enable=true
    Exponent=2
    Light Enabled=true
    Radius=500
   )
  ]
  rendererClasses [&NiagaraSpriteRendererProperties]
  renderers [{rendererClass=&NiagaraSpriteRendererProperties}]
 }
 AirGlow001 {
  bEnabled=false
  emitterUpdateScript [
   &EmitterState()
   &SpawnBurst_Instantaneous(Spawn Count=1, Use Spawn Probability=false)
  ]
  particleSpawnScript [&InitializeParticle(
    Color=$User.Color_01
    Lifetime=&RandomRangeFloat(Minimum=0.35, Maximum=0.35)
    Module.Write Color=true
    Module.Write Lifetime=true
    Module.Write Mass=true
    Module.Write Material Random=true
    Module.Write Position=false
    Module.Write Scale=false
    Module.Write SpriteRotation=false
    Module.Write SpriteSize=true
    Sprite Size=&Vector2DFromFloat(Value=&Multiply_Float(A=&RandomRangeFloat(Minimum=1200, Maximum=1200, Evaluation Type=::NewEnumerator1), B=$User.ImpactSize))
   )]
  particleUpdateScript [
   &ParticleState()
   &ScaleSpriteSize(Scale Factor=&Vector2DFromFloat(Value=&FloatFromCurve(
      FloatCurve=DataInterface() {propertyValues {Curve {keys [
          {arriveTangent=2.6083, interpMode=RCIM_Cubic, leaveTangent=0.4318, tangentMode=RCTM_User, time=0, value=0}
          {arriveTangent=1.3014, interpMode=RCIM_Cubic, leaveTangent=1.3014, tangentMode=RCTM_User, time=0.3, value=0.8}
          {arriveTangent=0.1331, leaveTangent=0.1331, time=1, value=1.2}
         ]}}}
      Scale Curve=1.25
     )))
   &Color(Color=&ColorFromCurve(ColorCurve=DataInterface() {propertyValues {
       AlphaCurve {keys [<at 0 1 RCIM_Cubic>,<at 1 1 RCIM_Cubic>]}
       BlueCurve {keys [<at 0.3144 0.9911 RCIM_Cubic>,<at 1 0.7741 RCIM_Cubic>]}
       GreenCurve {keys [<at 0.3144 0.6197 RCIM_Cubic>,<at 1 0.4 RCIM_Cubic>]}
       RedCurve {keys [<at 0.3144 0.242 RCIM_Cubic>,<at 1 1 RCIM_Cubic>]}
      }})) {enabled=false}
   &ScaleColor(
    Color Value To Scale=$Particles.Color
    Scale Alpha=&FloatFromCurve(FloatCurve=DataInterface() {propertyValues {Curve {keys [<at 0.5 1 RCIM_Cubic>,<at 1 0 RCIM_Cubic>]}}}, Scale Curve=1.1)
    Scale RGB=&VectorFromFloat(Value=&FloatFromCurve(
      FloatCurve=DataInterface() {propertyValues {Curve {keys [{interpMode=RCIM_Cubic, tangentMode=RCTM_User, time=0.5, value=1}, {interpMode=RCIM_Cubic, tangentMode=RCTM_User, time=1, value=0}]}}}
      Scale Curve=10
     ))
    ScaleA=true
    ScaleRGB=true
   )
   &CameraOffset(Camera Offset Amount=160)
   &FN_Radial_Material(Depth Fade=50, Radial Mask Power=20)
  ]
  rendererClasses [&NiagaraSpriteRendererProperties]
  renderers [{rendererClass=&NiagaraSpriteRendererProperties}]
 }
 Distortion {
  emitterUpdateScript [
   &EmitterScalability(
    Module.bApplySpawnCountScale=false
    Module.bApplySpawnCountScaleByDistanceFraction=false
    Module.bUseMaxDistance=false
    Module.bUseMinDistance=false
   ) {enabled=false}
   &EmitterLifeCycle(NextLoopDuration=4.4208) {enabled=false}
   &EmitterState()
   &SpawnBurst_Instantaneous(Spawn Count=1, Use Spawn Probability=false)
  ]
  particleSpawnScript [
   &InitializeParticle(
    Color=$User.Color_02
    Lifetime=&UniformRangedFloat(Minimum=0.35, Maximum=0.35)
    Mesh Scale=&VectorFromFloat(Value=&Multiply_Float(A=13, B=$User.ImpactSize))
    Module.Write Color=true
    Module.Write Lifetime=true
    Module.Write Mass=true
    Module.Write Material Random=true
    Module.Write Position=false
    Module.Write Scale=true
    Module.Write SpriteRotation=false
    Module.Write SpriteSize=false
    Position=&SimulationPosition()
   )
   /Niagara/Modules/Spawn/Location/SphereLocation(MaskSpawnGroup=false, Module.Override Randomness=false, Offset=<20 0 0>, Sphere Radius=0.01, Surface Only=false) {enabled=false}
  ]
  particleUpdateScript [
   &ParticleState()
   &Color() {enabled=false}
   &DynamicMaterialParameters(
    Index 0 Param 1=&Lerp_Float(
     A=&FloatFromCurve(FloatCurve=DataInterface() {propertyValues {Curve {keys [
          {interpMode=RCIM_Cubic, tangentMode=RCTM_User, time=0, value=1}
          {interpMode=RCIM_Cubic, tangentMode=RCTM_User, time=0.3, value=0.95}
          {arriveTangent=-0.0005, interpMode=RCIM_Cubic, leaveTangent=-0.0005, tangentMode=RCTM_User, time=0.6, value=0.95}
          {interpMode=RCIM_Cubic, tangentMode=RCTM_User, time=1, value=1}
         ]}}})
     B=-1
     Alpha=&FN_LOD_Distance_Remap(Start Distance=0, End Distance=0, Minimum Output=0, Maximum Output=0)
    )
    Index 0 Param 2=&FloatFromCurve(FloatCurve=DataInterface() {propertyValues {Curve {keys [<at 0 15>,<at 1 0>]}}})
    Module.Param0WriteEnabled=true
    Module.Param1WriteEnabled=true
    Module.Param2WriteEnabled=true
    Module.Param3WriteEnabled=true
   )
   &ScaleMeshSize(Scale Factor=&VectorFromFloat(Value=&FloatFromCurve(FloatCurve=DataInterface() {propertyValues {Curve {keys [
          {arriveTangent=4.1785, interpMode=RCIM_Cubic, leaveTangent=4.1785, tangentMode=RCTM_User, time=0, value=0}
          {arriveTangent=1.9532, interpMode=RCIM_Cubic, leaveTangent=1.9532, tangentMode=RCTM_User, time=0.2223, value=0.6}
          {arriveTangent=-0.0405, interpMode=RCIM_Cubic, leaveTangent=-0.0405, tangentMode=RCTM_User, time=1, value=1.25}
         ]}}})))
   &ScaleColor(
    Scale Alpha=&FloatFromCurve(
     FloatCurve=DataInterface() {propertyValues {Curve {keys [
         {interpMode=RCIM_Cubic, tangentMode=RCTM_User, time=0.0045, value=0.9397}
         {arriveTangent=-1.0036, interpMode=RCIM_Cubic, leaveTangent=-1.0036, tangentMode=RCTM_User, time=0.4647, value=0.5233}
         {arriveTangent=-2.4196, interpMode=RCIM_Cubic, leaveTangent=-2.4196, tangentMode=RCTM_User, time=1.0011, value=0.002}
        ]}}}
     Scale Curve=0.2
    )
    Scale RGB=&VectorFromCurve(VectorCurve=DataInterface() {propertyValues {XCurve {keys [<at 0 1>,<at 0.5 0.8>]}, YCurve {keys [<at 0 1>,<at 0.5 0.8>]}, ZCurve {keys [<at 0 1>,<at 0.5 0.8>]}}})
    ScaleA=true
    ScaleRGB=true
   )
  ]
  rendererClasses [&NiagaraMeshRendererProperties]
  renderers [{rendererClass=&NiagaraMeshRendererProperties, rendererIndex=0}]
 }
 FN_Core_Smoke_Sprite {
  emitterUpdateScript [
   &EmitterState()
   &SpawnBurst_Instantaneous(
    Spawn Count=&RandomRangeInt(Minimum=10, Maximum=12)
    Spawn Time=&Subtract_Float(A=$User.ExplosionDelay, B=0)
    Use Spawn Probability=false
   )
  ]
  particleSpawnScript [
   &InitializeParticle(
    Color=&UniformRangedLinearColor(Minimum=<rgba 0.2 0.185 0.17 1>, Maximum=<rgba 0.125 0.1156 0.1063 1>)
    Lifetime=&UniformRangedFloat(Minimum=0.8, Maximum=1.2)
    Module.Write Color=true
    Module.Write Lifetime=true
    Module.Write Mass=true
    Module.Write Material Random=true
    Module.Write Position=true
    Module.Write Scale=false
    Module.Write SpriteRotation=true
    Module.Write SpriteSize=true
    Sprite Rotation=&RandomRangeFloat(Minimum=0, Maximum=1)
    Sprite Size=&Vector2DFromFloat(Value=&RandomRangeFloat(Minimum=150, Maximum=225))
   )
   /SetVariables(Particles.InitSpriteRotRate=&RandomRangeFloat(Minimum=-2, Maximum=2)) {isSetParameters=true}
   &TorusLocation(
    Align Torus With Axis=false
    Large Radius=190
    MaskSpawnGroup=false
    Module.Override Randomness=false
    Surface Position=1
    Torus Mode=::NewEnumerator2
   )
   &AddVelocityFromPoint(
    Module.Use Falloff Distance=false
    Velocity Origin:Vector3f=&SimulationPosition()
    Velocity Strength=&RandomRangeFloat(Minimum=2000, Maximum=2400)
   )
  ]
  particleUpdateScript [
   &ParticleState()
   &SpriteRotationRate(Rotation Rate=&Multiply_Float(
     A=$Particles.InitSpriteRotRate
     B=&FloatFromCurve(
      FloatCurve=DataInterface() {propertyValues {Curve {keys [
          {arriveTangent=-4.0107, interpMode=RCIM_Cubic, leaveTangent=-4.0107, tangentMode=RCTM_User, time=-0.0046, value=1.0113}
          {arriveTangent=-0.2496, interpMode=RCIM_Cubic, leaveTangent=-0.2496, tangentMode=RCTM_User, time=0.5, value=0.16}
         ]}}}
      CurveIndex=$Emitter.Age
      Scale Curve=150
     )
    ))
   &Drag(
    Drag=&FloatFromCurve(
     FloatCurve=DataInterface() {propertyValues {Curve {keys [<at 0 3.5 RCIM_Cubic>,<at 0.5 8 RCIM_Cubic>]}}}
     Scale Curve=&RandomRangeFloat(Minimum=1, Maximum=1.5)
    )
    Use Linear Drag=true
    Use Rotational Drag=false
   )
   &GravityForce(Gravity=<0 0 -200>)
   &SolveForcesAndVelocity(Clamp Velocity=false, Limit Acceleration=false)
   &ScaleSpriteSize(Scale Factor=&Vector2DFromFloat(Value=&FloatFromCurve(FloatCurve=DataInterface() {propertyValues {Curve {keys [{arriveTangent=1.491, interpMode=RCIM_Cubic, leaveTangent=1.491, tangentMode=RCTM_User, time=0, value=1.5}, <at 1.0011 3 RCIM_Cubic>]}}})))
   &DynamicMaterialParameters(
    Index 0 Param 1=&FloatFromCurve(FloatCurve=DataInterface() {propertyValues {Curve {keys [<at 0 0 RCIM_Cubic>,<at 1 0.25>]}}})
    Index 0 Param 4=&FloatFromCurve(FloatCurve=DataInterface() {propertyValues {Curve {keys [<at 0 0 RCIM_Cubic>,<at 1 0.5>]}}}, Scale Curve=0.1)
    Module.Param0WriteEnabled=true
    Module.Param1WriteEnabled=false
    Module.Param2WriteEnabled=false
    Module.Param3WriteEnabled=true
   )
   &Color(Color=&LerpLinearColors(
     StartColor=$User.Color_02
     EndColor=$User.Color_01
     LerpFactor=&FloatFromCurve(
      FloatCurve=DataInterface() {propertyValues {Curve {keys [<at 0 1>,<at 1 0>]}}}
      CurveIndex=&NormalizeDistanceRange(
       Start Position:NiagaraPosition=$Particles.Position
       End Position:NiagaraPosition=/Niagara/DynamicInputs/Helpers/SimulationPosition()
       Distance:NiagaraFloat=400
       Invert Normalized Range:NiagaraBool=false
      )
     )
    ))
   &ScaleColor(
    Color Value To Scale=$Particles.Color
    Scale Alpha=&FloatFromCurve(
     FloatCurve=DataInterface() {propertyValues {Curve {keys [
         {arriveTangent=9.0725, interpMode=RCIM_Cubic, leaveTangent=9.0725, tangentMode=RCTM_User, time=0, value=0}
         {arriveTangent=-2.5249, interpMode=RCIM_Cubic, leaveTangent=-2.5249, tangentMode=RCTM_User, time=0.1408, value=0.9847}
         {arriveTangent=-0.9014, interpMode=RCIM_Cubic, leaveTangent=-0.9014, tangentMode=RCTM_User, time=0.3, value=0.2}
         {arriveTangent=-0.0405, interpMode=RCIM_Cubic, leaveTangent=-0.0405, tangentMode=RCTM_User, time=1, value=0}
        ]}}}
     Scale Curve=0.22
    )
    Scale RGB=&MakeVectorFromLinearColorRGB(Color=&MultiplyLinearColor(
      Linear Color=&FN_Select_Color_from_Time_Of_Day(Morning=<rgba 0.35 0.3 1.5 1>, Day=<rgba 0.6038 0.6038 1.5 1>, Evening=<rgba 0.2 0.15 1.5 1>, Night=<rgba 1 1 1 1>)
      Scale Factor=&MakeVector4_FromFloat(Float=1)
     ))
    ScaleA=true
    ScaleRGB=true
   )
   &CameraOffset(Camera Offset Amount=&RandomRangeFloat(Minimum=30, Maximum=50))
  ]
  rendererClasses [&NiagaraSpriteRendererProperties]
  renderers [
   {rendererClass=&NiagaraSpriteRendererProperties}
   {rendererClass=&NiagaraSpriteRendererProperties, rendererIndex=1}
  ]
 }
 FN_Mesh {
  bEnabled=false
  emitterUpdateScript [
   &EmitterState(Life Cycle Mode=::NewEnumerator1, Loop Behavior=::NewEnumerator1, Loop Duration=1, Scalability Mode=::NewEnumerator1, UseLoopDelay=false)
   &SpawnBurst_Instantaneous(Spawn Count=1, Use Spawn Probability=false)
  ]
  particleSpawnScript [
   &InitializeParticle(
    Color=<rgba 16 8 20 1>
    Lifetime=&Multiply_Float(A=0.4, B=$User.ExplosionLife)
    Mesh Scale=&Multiply_VectorByFloat(Vector=&VectorFromFloat(Value=6.5), Float=$User.ImpactSize)
    Module.Write Color=true
    Module.Write Lifetime=true
    Module.Write Mass=true
    Module.Write Material Random=true
    Module.Write Position=false
    Module.Write Scale=true
    Module.Write SpriteRotation=true
    Module.Write SpriteSize=false
    Position=/Niagara/Functions/Localspace/SimulationPosition()
   )
   /SetVariables(
    Particles.MeshOrientationNew:Quat4f=/Niagara/DynamicInputs/Rotation/RandomQuaternion()
    Particles.Scale2:Vector3f=&UniformRangedVector(
     Minimum=&VectorFromFloat(Value=1)
     Maximum=&VectorFromFloat(Value=2)
    )
   ) {isSetParameters=true}
   &DynamicMaterialParameters(
    Index 0 Param 1=&UniformRangedFloat(Minimum=0, Maximum=10)
    Index 0 Param 2=&UniformRangedFloat(Minimum=2, Maximum=5)
    Index 0 Param 3=&UniformRangedFloat(Minimum=0.5, Maximum=2)
    Index 0 Param 4=1
    Module.Param0WriteEnabled=true
    Module.Param1WriteEnabled=true
    Module.Param2WriteEnabled=true
    Module.Param3WriteEnabled=true
   )
   /Niagara/Modules/Spawn/Orientation/V2/InitialMeshOrientation(Use Orientation Vector=true, Use Rotation Vector=true) {enabled=false}
   /Niagara/Modules/Spawn/Orientation/V3/InitialMeshOrientation(Use Orientation Vector=true, Use Rotation Vector=true) {module=InitialMeshOrientation001}
  ]
  particleUpdateScript [
   &ParticleState()
   &CameraOffset(Camera Offset Amount=250) {module=CameraOffset001}
   &ScaleColor(
    Scale Alpha=&FloatFromCurve(FloatCurve=DataInterface() {propertyValues {Curve {keys [<at 0.6 1.0113>,<at 1.0011 0.002>]}}}, Scale Curve=0.025)
    Scale RGB=&VectorFromCurve(VectorCurve=DataInterface() {propertyValues {
       XCurve {keys [<at -0.0011 1.0067>,<at 0.9988 0>]}
       YCurve {keys [<at -0.0011 1.0067>,<at 0.9988 0>]}
       ZCurve {keys [<at -0.0011 1.0067>,<at 0.9988 0>]}
      }})
    ScaleA=true
    ScaleRGB=true
   )
   /SetVariables(
    Particles.DynamicMaterialParameter:Vector4f=&MakeFloat4FromVectorAndFloat("Vector (RGB)"=&VectorFromFloat(Value=1), "Float (Alpha)"=1)
    Particles.DynamicMaterialParameter1:Vector4f=&MakeFloat4FromVectorAndFloat(
     "Vector (RGB)"=&VectorFromFloat(Value=1)
     "Float (Alpha)"=&UniformRangedFloat(Minimum=0.5, Maximum=1)
    )
    Particles.DynamicMaterialParameter2:Vector4f=<1 1 1 0>
    Particles.DynamicMaterialParameter3:Vector4f=<1 1 1 0>
   ) {isSetParameters=true}
   &ScaleMeshSize(Scale Factor=&ScaleFloatByCurve(
     InitialValue:Vector3f=<1 1 1>
     ScaleCurve:NiagaraDataInterfaceVectorCurve=DataInterface() {propertyValues {XCurve {keys [<at 0 0.25>,<at 0.25 1>]}, YCurve {keys [<at 0 0.25>,<at 0.25 1>]}, ZCurve {keys [<at 0 0.25>,<at 0.25 1>]}}}
     CurveIndex:NiagaraFloat=$Particles.NormalizedAge
    ))
  ]
  rendererClasses [&NiagaraMeshRendererProperties]
  renderers [{rendererClass=&NiagaraMeshRendererProperties, rendererIndex=0}]
 }
 FN_RadialSprite {
  bEnabled=false
  emitterUpdateScript [
   &EmitterState()
   &SpawnRate(SpawnRate=12, Use Spawn Probability=false)
  ]
  particleSpawnScript [/Niagara/Modules/Spawn/Initialization/V2/InitializeParticle(Color Mode=::NewEnumerator0, Lifetime=0.5, UsePositionOffset=false, Write Material Random=true)]
  particleUpdateScript [
   &ParticleState()
   &FN_Radial_Material(Sprite Extrusion Speed=0.5)
  ]
  rendererClasses [&NiagaraSpriteRendererProperties]
  renderers [{rendererClass=&NiagaraSpriteRendererProperties}]
 }
 GroundRibbon {
  bEnabled=false
  emitterUpdateScript [
   &EmitterState()
   &SpawnBurst_Instantaneous(Spawn Count=20, Use Spawn Probability=false)
   &SpawnRate(SpawnRate=5, Use Spawn Probability=false) {enabled=false}
  ]
  particleSpawnScript [
   &InitializeParticle(
    Color=$User.Color_02
    Lifetime=&Multiply_Float(A=0.35, B=$User.ExplosionLife)
    Mass Max=2
    Mass Mode=::NewEnumerator3
    Sprite Size Max=<2 8>
    Sprite Size Min=<1.5 6>
    Sprite Size Mode=::NewEnumerator2
    UsePositionOffset=false
    Write Material Random=true
   ) {enabled=false}
   &InitializeRibbon(
    Color=&LerpLinearColors(StartColor=$User.Color_02, EndColor=$User.Color_03, LerpFactor=&RandomRangeFloat(Minimum=0, Maximum=1))
    Lifetime=&Multiply_Float(A=1, B=$User.ExplosionLife)
    Module.Write Color=true
    Module.Write Lifetime=true
    Module.Write Mass=true
    Module.Write Material Random=true
    Module.Write Position=false
    Module.Write RibbonTwist=false
    Module.Write RibbonWidth=true
    Ribbon Width=&RandomRangeFloat(Minimum=32, Maximum=42)
   )
   &AddVelocityInCone(
    Cone Angle=48
    Cone Axis=<0.9 0 0.4>
    Module.Use Velocity Falloff On Cone Axis=true
    Velocity Distribution Along Cone Axis=0.75
    Velocity Falloff Away From Cone Axis=0.333
    Velocity Strength=&RandomRangeFloat(Minimum=350, Maximum=750)
   ) {enabled=false}
   &CylinderLocation(
    Cylinder Height=0
    Cylinder Radius=&Multiply_Float(A=$User.ImpactSize, B=50)
    MaskSpawnGroup=false
    Module.Enable Lathe Profile=false
    Module.Override Randomness=false
    Module.Surface Only=true
    Surface Only Band Thickness=0
    Use Endcaps In Surface Only Mode=false
   )
   &InheritVelocity(Module.Limit Speed=true, Velocity Limit=250, Velocity Scale=<-1 -1 -1>) {enabled=false}
   /Niagara/Modules/Spawn/Location/V2/SphereLocation(MaskSpawnGroup=false, Override Seed=false, Surface Only=false, UseNonUniformScale=false, UseOffset=false, UseSphereOrientation=false) {enabled=false}
   &FN_SetSpawnGroup(Number of Spawn Groups=1, Set Ribbon ID and Ribbon Link Order=true)
   &AddVelocityFromPoint(Module.Use Falloff Distance=false, Velocity Strength=&RandomRangeFloat(Minimum=1200, Maximum=1600))
  ]
  particleUpdateScript [
   &ParticleState()
   &GravityForce(Gravity=<0 0 -200>) {enabled=false}
   &Drag(
    Drag=&FloatFromCurve(FloatCurve=DataInterface() {propertyValues {Curve {keys [{interpMode=RCIM_Cubic, tangentMode=RCTM_User, time=0, value=0}, {interpMode=RCIM_Cubic, tangentMode=RCTM_User, time=0.125, value=2.5}]}}})
    Use Linear Drag=true
    Use Rotational Drag=false
   )
   &CurlNoiseForce(
    Noise Frequency=150
    Noise Strength=&Multiply_Float(
     A=3500
     B=&FloatFromCurve(FloatCurve=DataInterface() {propertyValues {Curve {keys [<at 0 1>,<at 0.5 0>]}}})
    )
    Pan Noise=false
   )
   &SolveForcesAndVelocity(Clamp Velocity=false, Limit Acceleration=false)
   &ScaleSpriteSizeBySpeed(Max Scale Factor=<1.2 5>, Velocity Threshold=1500) {enabled=false}
   &ScaleSpriteSize(Scale Factor=&Vector2DFromFloat(Value=&FloatFromCurve(FloatCurve=DataInterface() {propertyValues {Curve {keys [
          {arriveTangent=9.0725, interpMode=RCIM_Cubic, leaveTangent=9.0725, tangentMode=RCTM_User, time=0, value=0}
          {arriveTangent=-2.5249, interpMode=RCIM_Cubic, leaveTangent=-2.5249, tangentMode=RCTM_User, time=0.1408, value=0.9847}
          {arriveTangent=-0.0405, interpMode=RCIM_Cubic, leaveTangent=-0.0405, tangentMode=RCTM_User, time=1, value=0}
         ]}}}))) {enabled=false}
   &ScaleColor(
    Scale Alpha=&FloatFromCurve(
     FloatCurve=DataInterface() {propertyValues {Curve {keys [
         {interpMode=RCIM_Cubic, tangentMode=RCTM_User, time=0, value=0}
         {interpMode=RCIM_Cubic, tangentMode=RCTM_User, time=0.25, value=1}
         {arriveTangent=-1.3333, interpMode=RCIM_Cubic, leaveTangent=-1.3333, time=0.4, value=0.25}
         {tangentMode=RCTM_User, time=1, value=0}
        ]}}}
     Scale Curve=&RandomRangeFloat(Minimum=1, Maximum=2)
    )
    ScaleA=true
    ScaleRGB=true
   )
   &ScaleRibbonWidth(Ribbon Width Scale=&FloatFromCurve(
     FloatCurve=DataInterface() {propertyValues {Curve {keys [{tangentMode=RCTM_User, time=0.9, value=0}, {tangentMode=RCTM_User, time=1, value=3}]}}}
     CurveIndex=&NormalizeDistanceRange(
      Start Position:NiagaraPosition=$Particles.Position
      End Position:NiagaraPosition=$Engine.Owner.Position
      Distance:NiagaraFloat=1
      Invert Normalized Range:NiagaraBool=false
     )
    )) {enabled=false}
   &ScaleColor(Color Value To Scale=$Particles.Color, Scale Alpha=220, ScaleA=true, ScaleRGB=true) {enabled=false, module=ScaleColor001}
  ]
  rendererClasses [&NiagaraSpriteRendererProperties, &NiagaraRibbonRendererProperties]
  renderers [
   {rendererClass=&NiagaraSpriteRendererProperties}
   {rendererClass=&NiagaraRibbonRendererProperties, rendererIndex=1}
  ]
 }
 ImpactRingsNeg {
  bEnabled=false
  emitterUpdateScript [
   &EmitterState()
   &SpawnBurst_Instantaneous(Spawn Count=5, Use Spawn Probability=false)
  ]
  particleSpawnScript [
   &InitializeParticle(
    Color=$User.Color_02
    Lifetime=0.5
    Mesh Scale Mode=::NewEnumerator3
    Mesh Uniform Scale=&Multiply_Float(A=$User.ImpactSize, B=5)
    Position Offset=<0 0 5>
    UsePositionOffset=true
    Write Material Random=true
   )
   &InitialMeshOrientation(
    Rotation=&RandomRangeVector(Minimum=<0 0 0>, Maximum=<0 0 1>)
    Use Orientation Vector=true
    Use Rotation Vector=true
   )
   &AddVelocity(Velocity=&Multiply_VectorByFloat(Vector=<0 0 0>, Float=$User.RingSize)) {enabled=false}
   /SetVariables(Particles.PositionEnd:Vector3f=&Add_Vector(
     A=/Niagara/DynamicInputs/Helpers/SimulationPosition()
     B=&MakeVector(
      X=0
      Y=0
      Z=&Multiply_Float(
       A=&Multiply_Float(A=-100, B=$User.ImpactSize)
       B=&MakeFloatFromInt(Int=&ReturnParticleIDAttributes())
      )
     )
    )) {isSetParameters=true, module=SetVariables_#1}
  ]
  particleUpdateScript [
   &ParticleState()
   &SolveRotationalForcesAndVelocity() {enabled=false}
   /SetVariables(Particles.SizeWave:NiagaraFloat=&Waveform(
     Global Input Position=&Multiply_Float(A=$Engine.Time, B=1)
     Global Phase=$Particles.NormalizedAge
     "Waveform [1]"=::NewEnumerator0
     "[1] Phase Scale"=1.75
     "[1] Phase"=$Particles.NormalizedAge
     "[1] Frequency"=&Lerp_Float(A=0.4, B=1.1, Alpha=$User.ADS)
     "[1] Amplitude Min/Max"=<0 1>
     "[1] Scale"=0.5
     "[1] Bias"=0.5
     "[2] Phase"=$Emitter.Age
     "[3] Phase"=$Emitter.Age
    )) {enabled=false, isSetParameters=true, module=SetVariables_#2}
   &SolveForcesAndVelocity(Clamp Velocity=false, Limit Acceleration=false)
   &ScaleColor(
    Scale Alpha=2
    Scale RGB=&Multiply_VectorByFloat(Vector=<1 1 1>, Float=&Lerp_Float(A=2, B=3, Alpha=$User.ADS))
    ScaleA=true
    ScaleRGB=true
   )
   &DynamicMaterialParameters(
    Index 0 Param 1=&FloatFromCurve(
     FloatCurve=DataInterface() {propertyValues {Curve {keys [{interpMode=RCIM_Cubic, tangentMode=RCTM_User, time=0, value=1}, {interpMode=RCIM_Cubic, tangentMode=RCTM_User, time=1, value=0}]}}}
     Scale Curve=100
    )
    Module.Param0WriteEnabled=true
    Module.Param1WriteEnabled=true
    Module.Param2WriteEnabled=true
    Module.Param3WriteEnabled=false
   )
   &ScaleMeshSize(Scale Factor=&Multiply_VectorByFloat(
     Vector=&Multiply_VectorByFloat(
      Vector=&MakeVector(X=$Particles.SizeWave, Y=$Particles.SizeWave, Z=1)
      Float=&FloatFromCurve(FloatCurve=DataInterface() {propertyValues {Curve {keys [
           {interpMode=RCIM_Cubic, tangentMode=RCTM_User, time=0, value=1}
           {arriveTangent=-0.5603, interpMode=RCIM_Cubic, tangentMode=RCTM_Break, time=1, value=0.8}
          ]}}})
     )
     Float=$User.RingSize
    )) {enabled=false}
   &ScaleMeshSize(Scale Factor=&VectorFromFloat(Value=&FloatFromCurve(FloatCurve=DataInterface() {propertyValues {Curve {keys [
          {arriveTangent=4.1785, interpMode=RCIM_Cubic, leaveTangent=4.1785, tangentMode=RCTM_User, time=0, value=0.25}
          {arriveTangent=0.4581, interpMode=RCIM_Cubic, leaveTangent=0.4581, tangentMode=RCTM_User, time=0.2223, value=0.9}
          {arriveTangent=-0.0405, interpMode=RCIM_Cubic, leaveTangent=-0.0405, tangentMode=RCTM_User, time=1, value=1}
         ]}}}))) {enabled=false, module=ScaleMeshSize001}
   &ScaleMeshSize(Scale Factor=&VectorFromFloat(Value=&FloatFromCurve(FloatCurve=DataInterface() {propertyValues {Curve {keys [
          <at 0 1 RCIM_Cubic>
          {arriveTangent=0.25, interpMode=RCIM_Cubic, leaveTangent=0.25, time=0.1, value=0.95}
          <at 0.2 1.05 RCIM_Cubic>
          {arriveTangent=-0.25, interpMode=RCIM_Cubic, leaveTangent=-0.25, time=0.3, value=0.95}
          {arriveTangent=0.1167, interpMode=RCIM_Cubic, leaveTangent=0.1167, time=0.4, value=1}
          {arriveTangent=0.125, interpMode=RCIM_Cubic, leaveTangent=0.125, time=0.6, value=0.985}
          {arriveTangent=0.0375, interpMode=RCIM_Cubic, leaveTangent=0.0375, time=0.8, value=1.05}
          <at 1 1 RCIM_Cubic>
         ]}}}))) {enabled=false, module=ScaleMeshSize002}
   &ScaleMeshSize(Scale Factor=&VectorFromFloat(Value=&RandomRangeFloat(Minimum=0.99, Maximum=1.01, Evaluation Type=::NewEnumerator1))) {module=ScaleMeshSize003}
   &ScaleMeshSize(Scale Factor=&VectorFromFloat(Value=&FloatFromCurve(FloatCurve=DataInterface() {propertyValues {Curve {keys [
          {arriveTangent=4.1785, interpMode=RCIM_Cubic, leaveTangent=4.1785, tangentMode=RCTM_User, time=0, value=0.05}
          {arriveTangent=1.5995, interpMode=RCIM_Cubic, leaveTangent=1.5995, tangentMode=RCTM_User, time=0.2, value=0.65}
          {arriveTangent=-0.0405, interpMode=RCIM_Cubic, leaveTangent=-0.0405, tangentMode=RCTM_User, time=1, value=1}
         ]}}}))) {module=ScaleMeshSize004}
   &ScaleMeshSize(Scale Factor=&VectorFromFloat(Value=&FloatFromCurve(
      FloatCurve=DataInterface() {propertyValues {Curve {keys [
          {interpMode=RCIM_Cubic, tangentMode=RCTM_User, time=0, value=1}
          {arriveTangent=-2.6029, interpMode=RCIM_Cubic, leaveTangent=-2.6029, tangentMode=RCTM_User, time=0.95, value=0}
         ]}}}
      CurveIndex=&Divide_Float(
       A=&MakeFloatFromInt(Int=&ReturnParticleIDAttributes())
       B=/Game/Effects/Niagara/DynamicInputs/Attributes/FN_GetEmitterParticleCount(
        AttributeReader:NiagaraDataInterfaceParticleRead=DataInterface() {propertyValues {EmitterBinding {bindingMode=Other, emitterName=ImpactRingsNeg}}}
        Multiplier:NiagaraFloat=1
       )
      )
     ))) {module=ScaleMeshSize005}
   &ScaleColor(
    Color Value To Scale=$Particles.Color
    Scale Alpha=&FloatFromCurve(
     FloatCurve=DataInterface() {propertyValues {Curve {keys [{interpMode=RCIM_Cubic, tangentMode=RCTM_User, time=0, value=1}, {interpMode=RCIM_Cubic, tangentMode=RCTM_User, time=1, value=0}]}}}
     Scale Curve=&Lerp_Float(
      A=0.1
      B=-0.01
      Alpha=&Divide_Float(
       A=&MakeFloatFromInt(Int=&ReturnParticleIDAttributes())
       B=&FN_GetEmitterParticleCount(
        AttributeReader:NiagaraDataInterfaceParticleRead=DataInterface() {propertyValues {EmitterBinding {bindingMode=Other, emitterName=ImpactRingsNeg}}}
        Multiplier:NiagaraFloat=1
       )
      )
     )
    )
    ScaleA=true
    ScaleRGB=true
   ) {module=ScaleColor001}
   /SetVariables(Particles.Position=&Lerp_Vector(
     A=&SimulationPosition()
     B=$Particles.Initial.PositionEnd
     Alpha=&FloatFromCurve(FloatCurve=DataInterface() {propertyValues {Curve {keys [
          {arriveTangent=1.5939, interpMode=RCIM_Cubic, leaveTangent=1.5939, tangentMode=RCTM_User, time=0, value=0.5}
          {interpMode=RCIM_Cubic, tangentMode=RCTM_User, time=1, value=1}
         ]}}})
    )) {isSetParameters=true, module=SetVariables_#3}
  ]
  rendererClasses [&NiagaraMeshRendererProperties]
  renderers [
   {rendererClass=&NiagaraMeshRendererProperties, rendererIndex=0}
   {rendererClass=&NiagaraMeshRendererProperties, rendererIndex=1}
  ]
 }
 ImpactRingsPos {
  bEnabled=false
  emitterUpdateScript [
   &EmitterState()
   &SpawnBurst_Instantaneous(Spawn Count=5, Use Spawn Probability=false)
  ]
  particleSpawnScript [
   &InitializeParticle(
    Color=$User.Color_02
    Lifetime=0.5
    Mesh Scale Mode=::NewEnumerator3
    Mesh Uniform Scale=&Multiply_Float(A=$User.ImpactSize, B=5)
    Position Offset=<0 0 5>
    UsePositionOffset=true
    Write Material Random=true
   )
   &InitialMeshOrientation(
    Rotation=&RandomRangeVector(Minimum=<0 0 0>, Maximum=<0 0 1>)
    Use Orientation Vector=true
    Use Rotation Vector=true
   )
   &AddVelocity(Velocity=&Multiply_VectorByFloat(Vector=<0 0 0>, Float=$User.RingSize)) {enabled=false}
   /SetVariables(Particles.PositionEnd:Vector3f=&Add_Vector(
     A=&SimulationPosition()
     B=&MakeVector(
      X=0
      Y=0
      Z=&Multiply_Float(
       A=&Multiply_Float(A=100, B=$User.ImpactSize)
       B=&MakeFloatFromInt(Int=&ReturnParticleIDAttributes())
      )
     )
    )) {isSetParameters=true, module=SetVariables_#1}
  ]
  particleUpdateScript [
   &ParticleState()
   &SolveRotationalForcesAndVelocity() {enabled=false}
   /SetVariables(Particles.SizeWave:NiagaraFloat=&Waveform(
     Global Input Position=&Multiply_Float(A=$Engine.Time, B=1)
     Global Phase=$Particles.NormalizedAge
     "Waveform [1]"=::NewEnumerator0
     "[1] Phase Scale"=1.75
     "[1] Phase"=$Particles.NormalizedAge
     "[1] Frequency"=&Lerp_Float(A=0.4, B=1.1, Alpha=$User.ADS)
     "[1] Amplitude Min/Max"=<0 1>
     "[1] Scale"=0.5
     "[1] Bias"=0.5
     "[2] Phase"=$Emitter.Age
     "[3] Phase"=$Emitter.Age
    )) {enabled=false, isSetParameters=true, module=SetVariables_#2}
   &SolveForcesAndVelocity(Clamp Velocity=false, Limit Acceleration=false)
   &ScaleColor(
    Scale Alpha=2
    Scale RGB=&Multiply_VectorByFloat(Vector=<1 1 1>, Float=&Lerp_Float(A=2, B=3, Alpha=$User.ADS))
    ScaleA=true
    ScaleRGB=true
   )
   &DynamicMaterialParameters(
    Index 0 Param 1=&FloatFromCurve(
     FloatCurve=DataInterface() {propertyValues {Curve {keys [{interpMode=RCIM_Cubic, tangentMode=RCTM_User, time=0, value=1}, {interpMode=RCIM_Cubic, tangentMode=RCTM_User, time=1, value=0}]}}}
     Scale Curve=100
    )
    Module.Param0WriteEnabled=true
    Module.Param1WriteEnabled=true
    Module.Param2WriteEnabled=true
    Module.Param3WriteEnabled=false
   )
   &ScaleMeshSize(Scale Factor=&Multiply_VectorByFloat(
     Vector=&Multiply_VectorByFloat(
      Vector=&MakeVector(X=$Particles.SizeWave, Y=$Particles.SizeWave, Z=1)
      Float=&FloatFromCurve(FloatCurve=DataInterface() {propertyValues {Curve {keys [
           {interpMode=RCIM_Cubic, tangentMode=RCTM_User, time=0, value=1}
           {arriveTangent=-0.5603, interpMode=RCIM_Cubic, tangentMode=RCTM_Break, time=1, value=0.8}
          ]}}})
     )
     Float=$User.RingSize
    )) {enabled=false}
   &ScaleMeshSize(Scale Factor=&VectorFromFloat(Value=&FloatFromCurve(FloatCurve=DataInterface() {propertyValues {Curve {keys [
          {arriveTangent=4.1785, interpMode=RCIM_Cubic, leaveTangent=4.1785, tangentMode=RCTM_User, time=0, value=0.25}
          {arriveTangent=0.4581, interpMode=RCIM_Cubic, leaveTangent=0.4581, tangentMode=RCTM_User, time=0.2223, value=0.9}
          {arriveTangent=-0.0405, interpMode=RCIM_Cubic, leaveTangent=-0.0405, tangentMode=RCTM_User, time=1, value=1}
         ]}}}))) {enabled=false, module=ScaleMeshSize001}
   &ScaleMeshSize(Scale Factor=&VectorFromFloat(Value=&FloatFromCurve(FloatCurve=DataInterface() {propertyValues {Curve {keys [
          <at 0 1 RCIM_Cubic>
          {arriveTangent=0.25, interpMode=RCIM_Cubic, leaveTangent=0.25, time=0.1, value=0.95}
          <at 0.2 1.05 RCIM_Cubic>
          {arriveTangent=-0.25, interpMode=RCIM_Cubic, leaveTangent=-0.25, time=0.3, value=0.95}
          {arriveTangent=0.1167, interpMode=RCIM_Cubic, leaveTangent=0.1167, time=0.4, value=1}
          {arriveTangent=0.125, interpMode=RCIM_Cubic, leaveTangent=0.125, time=0.6, value=0.985}
          {arriveTangent=0.0375, interpMode=RCIM_Cubic, leaveTangent=0.0375, time=0.8, value=1.05}
          <at 1 1 RCIM_Cubic>
         ]}}}))) {enabled=false, module=ScaleMeshSize002}
   &ScaleMeshSize(Scale Factor=&VectorFromFloat(Value=&RandomRangeFloat(Minimum=0.99, Maximum=1.01, Evaluation Type=::NewEnumerator1))) {module=ScaleMeshSize003}
   &ScaleMeshSize(Scale Factor=&VectorFromFloat(Value=&FloatFromCurve(FloatCurve=DataInterface() {propertyValues {Curve {keys [
          {arriveTangent=4.1785, interpMode=RCIM_Cubic, leaveTangent=4.1785, tangentMode=RCTM_User, time=0, value=0.05}
          {arriveTangent=1.5995, interpMode=RCIM_Cubic, leaveTangent=1.5995, tangentMode=RCTM_User, time=0.2, value=0.65}
          {arriveTangent=-0.0405, interpMode=RCIM_Cubic, leaveTangent=-0.0405, tangentMode=RCTM_User, time=1, value=1}
         ]}}}))) {module=ScaleMeshSize004}
   &ScaleMeshSize(Scale Factor=&VectorFromFloat(Value=&FloatFromCurve(
      FloatCurve=DataInterface() {propertyValues {Curve {keys [
          {interpMode=RCIM_Cubic, tangentMode=RCTM_User, time=0, value=1}
          {arriveTangent=-2.6029, interpMode=RCIM_Cubic, leaveTangent=-2.6029, tangentMode=RCTM_User, time=0.95, value=0}
         ]}}}
      CurveIndex=&Divide_Float(
       A=&MakeFloatFromInt(Int=&ReturnParticleIDAttributes())
       B=&FN_GetEmitterParticleCount(
        AttributeReader:NiagaraDataInterfaceParticleRead=DataInterface() {propertyValues {EmitterBinding {bindingMode=Other, emitterName=ImpactRingsPos}}}
        Multiplier:NiagaraFloat=1
       )
      )
     ))) {module=ScaleMeshSize005}
   &ScaleColor(
    Color Value To Scale=$Particles.Color
    Scale Alpha=&FloatFromCurve(
     FloatCurve=DataInterface() {propertyValues {Curve {keys [{interpMode=RCIM_Cubic, tangentMode=RCTM_User, time=0, value=1}, {interpMode=RCIM_Cubic, tangentMode=RCTM_User, time=1, value=0}]}}}
     Scale Curve=&Lerp_Float(
      A=0.1
      B=-0.01
      Alpha=&Divide_Float(
       A=&MakeFloatFromInt(Int=&ReturnParticleIDAttributes())
       B=&FN_GetEmitterParticleCount(
        AttributeReader:NiagaraDataInterfaceParticleRead=DataInterface() {propertyValues {EmitterBinding {bindingMode=Other, emitterName=ImpactRingsPos}}}
        Multiplier:NiagaraFloat=1
       )
      )
     )
    )
    ScaleA=true
    ScaleRGB=true
   ) {module=ScaleColor001}
   /SetVariables(Particles.Position=&Lerp_Vector(
     A=&SimulationPosition()
     B=$Particles.Initial.PositionEnd
     Alpha=&FloatFromCurve(FloatCurve=DataInterface() {propertyValues {Curve {keys [
          {arriveTangent=1.5939, interpMode=RCIM_Cubic, leaveTangent=1.5939, tangentMode=RCTM_User, time=0, value=0.5}
          {interpMode=RCIM_Cubic, tangentMode=RCTM_User, time=1, value=1}
         ]}}})
    )) {isSetParameters=true, module=SetVariables_#3}
  ]
  rendererClasses [&NiagaraMeshRendererProperties]
  renderers [
   {rendererClass=&NiagaraMeshRendererProperties, rendererIndex=0}
   {rendererClass=&NiagaraMeshRendererProperties, rendererIndex=1}
  ]
 }
 InnerGlowSphere {
  bEnabled=false
  emitterUpdateScript [
   &EmitterState()
   &SpawnBurst_Instantaneous(Spawn Count=1, Spawn Time=$User.ExplosionDelay, Use Spawn Probability=false)
  ]
  particleSpawnScript [&InitializeParticle(
    Color=&MultiplyLinearColor(Linear Color=$User.Color_01, Scale Factor=&MakeVector4_FromLinearColor(LinearColor=<rgba 1 1 1 1>))
    Lifetime=&Multiply_Float(A=0.5, B=$User.ExplosionLife)
    Mesh Scale Mode=::NewEnumerator3
    Mesh Uniform Scale=&Multiply_Float(A=11.5, B=$User.ImpactSize)
    UsePositionOffset=false
    Write Material Random=true
   ) {module=InitializeParticle001}]
  particleUpdateScript [
   &ParticleState()
   &ScaleColor(
    Scale Alpha=&FloatFromCurve(
     FloatCurve=DataInterface() {propertyValues {Curve {keys [
         {interpMode=RCIM_Cubic, tangentMode=RCTM_User, time=0, value=0}
         {interpMode=RCIM_Cubic, tangentMode=RCTM_User, time=0.5, value=1}
         {tangentMode=RCTM_User, time=1, value=0}
        ]}}}
     Scale Curve=0.2
    )
    ScaleA=true
    ScaleRGB=true
   )
   &DynamicMaterialParameters(
    Index 0 Param 1=1
    Index 0 Param 2=1
    Index 0 Param 3=1
    Index 0 Param 4=1
    Module.Param0WriteEnabled=true
    Module.Param1WriteEnabled=true
    Module.Param2WriteEnabled=true
    Module.Param3WriteEnabled=true
   )
   &ScaleMeshSize(Scale Factor=&VectorFromFloat(Value=&FloatFromCurve(FloatCurve=DataInterface() {propertyValues {Curve {keys [
          {arriveTangent=4.1785, interpMode=RCIM_Cubic, leaveTangent=4.1785, tangentMode=RCTM_User, time=0, value=0}
          {arriveTangent=0.9995, interpMode=RCIM_Cubic, leaveTangent=0.9995, tangentMode=RCTM_User, time=0.15, value=0.8}
          {arriveTangent=-0.0405, interpMode=RCIM_Cubic, leaveTangent=-0.0405, tangentMode=RCTM_User, time=1, value=1}
         ]}}})))
  ]
  rendererClasses [&NiagaraMeshRendererProperties]
  renderers [{rendererClass=&NiagaraMeshRendererProperties, rendererIndex=0}]
 }
 InnerNoiseSphere {
  emitterUpdateScript [
   &EmitterState()
   &SpawnBurst_Instantaneous(Spawn Count=1, Spawn Time=$User.ExplosionDelay, Use Spawn Probability=false)
  ]
  particleSpawnScript [
   &InitializeParticle(
    Lifetime=&Multiply_Float(A=0.3, B=$User.ExplosionLife)
    Mesh Scale Mode=::NewEnumerator3
    Mesh Uniform Scale=&Multiply_Float(A=5, B=$User.ImpactSize)
    UsePositionOffset=false
    Write Material Random=true
   ) {module=InitializeParticle001}
   &InitialMeshOrientation(Rotation=<0 0.25 0>, Rotation Coordinate Space=::NewEnumerator2, Use Orientation Vector=true, Use Rotation Vector=true) {enabled=false}
  ]
  particleUpdateScript [
   &ParticleState()
   &ScaleColor(
    Scale Alpha=&FloatFromCurve(FloatCurve=DataInterface() {propertyValues {Curve {keys [<at 0 1>, {interpMode=RCIM_Cubic, tangentMode=RCTM_User, time=0.1, value=1}, {interpMode=RCIM_Cubic, tangentMode=RCTM_User, time=1, value=0}]}}})
    Scale RGB=&MakeVectorFromLinearColorRGB(Color=&MultiplyLinearColor(
      Linear Color=&FN_Select_Color_from_Time_Of_Day(Morning=<rgba 0.4 0.8 1 1>, Day=<rgba 1 1 1 1>, Evening=<rgba 0.4 0.8 1 1>, Night=<rgba 1 1 0.5 1>)
      Scale Factor=&MakeVector4_FromFloat(Float=2.5)
     ))
    ScaleA=true
    ScaleRGB=true
   )
   &DynamicMaterialParameters(
    Index 0 Param 1=1
    Index 0 Param 2=1
    Index 0 Param 3=1
    Index 0 Param 4=1
    Module.Param0WriteEnabled=true
    Module.Param1WriteEnabled=true
    Module.Param2WriteEnabled=true
    Module.Param3WriteEnabled=true
   )
   &ScaleMeshSize(Scale Factor=&VectorFromFloat(Value=&FloatFromCurve(FloatCurve=DataInterface() {propertyValues {Curve {keys [
          {arriveTangent=4.1785, interpMode=RCIM_Cubic, leaveTangent=4.1785, tangentMode=RCTM_User, time=0, value=0.05}
          {arriveTangent=0.9995, interpMode=RCIM_Cubic, leaveTangent=0.9995, tangentMode=RCTM_User, time=0.3, value=0.6}
          {arriveTangent=-0.0405, interpMode=RCIM_Cubic, leaveTangent=-0.0405, tangentMode=RCTM_User, time=1, value=1}
         ]}}})))
   &CameraOffset(Camera Offset Amount=250)
  ]
  rendererClasses [&NiagaraMeshRendererProperties]
  renderers [{rendererClass=&NiagaraMeshRendererProperties, rendererIndex=0}]
 }
 InnerRadialSphere {
  emitterUpdateScript [
   &EmitterState()
   &SpawnBurst_Instantaneous(Spawn Count=1, Spawn Time=$User.ExplosionDelay, Use Spawn Probability=false)
  ]
  particleSpawnScript [&InitializeParticle(
    Color=&MultiplyLinearColor(Linear Color=$User.Color_01, Scale Factor=&MakeVector4_FromLinearColor(LinearColor=<rgba 1 1 1 0.1>))
    Lifetime=&Multiply_Float(A=0.5, B=$User.ExplosionLife)
    Mesh Scale Mode=::NewEnumerator3
    Mesh Uniform Scale=&Multiply_Float(A=11.5, B=$User.ImpactSize)
    UsePositionOffset=false
    Write Material Random=true
   ) {module=InitializeParticle001}]
  particleUpdateScript [
   &ParticleState()
   &ScaleColor(
    Scale Alpha=&FloatFromCurve(
     FloatCurve=DataInterface() {propertyValues {Curve {keys [
         {interpMode=RCIM_Cubic, tangentMode=RCTM_User, time=0, value=1}
         {interpMode=RCIM_Cubic, tangentMode=RCTM_User, time=0.1, value=1}
         {interpMode=RCIM_Cubic, tangentMode=RCTM_User, time=1, value=0}
        ]}}}
     Scale Curve=&FN_Select_Float_from_Time_Of_Day(Morning=0.6, Day=0.6, Evening=0.6, Night=0.6)
    )
    Scale RGB=&MakeVectorFromLinearColorRGB(Color=&MultiplyLinearColor(
      Linear Color=&FN_Select_Color_from_Time_Of_Day(Morning=<rgba 0.2 0.5 1 1>, Day=<rgba 1 1 1 1>, Evening=<rgba 0.2 0.5 1 1>, Night=<rgba 1 1 0.5 1>)
      Scale Factor=&MakeVector4_FromFloat(Float=&FloatFromCurve(
        FloatCurve=DataInterface() {propertyValues {Curve {keys [
            {interpMode=RCIM_Cubic, tangentMode=RCTM_User, time=0, value=15}
            {interpMode=RCIM_Cubic, tangentMode=RCTM_User, time=0.2, value=1}
            {interpMode=RCIM_Cubic, tangentMode=RCTM_User, time=1, value=1}
           ]}}}
        Scale Curve=2.5
       ))
     ))
    ScaleA=true
    ScaleRGB=true
   )
   &DynamicMaterialParameters(
    Index 0 Param 1=1
    Index 0 Param 2=1
    Index 0 Param 3=1
    Index 0 Param 4=1
    Module.Param0WriteEnabled=true
    Module.Param1WriteEnabled=true
    Module.Param2WriteEnabled=true
    Module.Param3WriteEnabled=true
   )
   &ScaleMeshSize(Scale Factor=&VectorFromFloat(Value=&FloatFromCurve(
      FloatCurve=DataInterface() {propertyValues {Curve {keys [
          {arriveTangent=4.1785, interpMode=RCIM_Cubic, leaveTangent=4.1785, tangentMode=RCTM_User, time=0, value=0.2}
          {arriveTangent=0.9995, interpMode=RCIM_Cubic, leaveTangent=0.9995, tangentMode=RCTM_User, time=0.15, value=0.8}
          {arriveTangent=-0.0405, interpMode=RCIM_Cubic, leaveTangent=-0.0405, tangentMode=RCTM_User, time=1, value=1}
         ]}}}
      Scale Curve=1.5
     )))
   &CameraOffset(Camera Offset Amount=-20)
  ]
  rendererClasses [&NiagaraMeshRendererProperties]
  renderers [{rendererClass=&NiagaraMeshRendererProperties, rendererIndex=0}]
 }
 Light {
  emitterUpdateScript [
   &EmitterState()
   &SpawnBurst_Instantaneous(Spawn Count=1, Use Spawn Probability=false)
  ]
  particleSpawnScript [/Niagara/Modules/Spawn/Initialization/InitializeParticle(
    Color=$User.Color_01
    Lifetime=&RandomRangeFloat(Minimum=0.28, Maximum=0.3)
    Module.Write Color=true
    Module.Write Lifetime=true
    Module.Write Mass=true
    Module.Write Material Random=true
    Module.Write Position=true
    Module.Write Scale=false
    Module.Write SpriteRotation=false
    Module.Write SpriteSize=true
    Position=&Add_Vector(A=&SimulationPosition(), B=<0 0 2.5>)
    Sprite Size=&Vector2DFromFloat(Value=&Multiply_Float(A=&RandomRangeFloat(Minimum=750, Maximum=850, Evaluation Type=::NewEnumerator1), B=$User.ImpactSize))
   )]
  particleUpdateScript [
   &ParticleState()
   &Color(Color=&ColorFromCurve(ColorCurve=DataInterface() {propertyValues {
       AlphaCurve {keys [<at 0 1 RCIM_Cubic>,<at 1 1 RCIM_Cubic>]}
       BlueCurve {keys [<at 0.3144 0.9911 RCIM_Cubic>,<at 1 0.7741 RCIM_Cubic>]}
       GreenCurve {keys [<at 0.3144 0.6197 RCIM_Cubic>,<at 1 0.4 RCIM_Cubic>]}
       RedCurve {keys [<at 0.3144 0.242 RCIM_Cubic>,<at 1 1 RCIM_Cubic>]}
      }})) {enabled=false}
   &ScaleColor(
    Color Value To Scale=$Particles.Color
    Scale Alpha=&FloatFromCurve(FloatCurve=DataInterface() {propertyValues {Curve {keys [<at 0 1 RCIM_Cubic>,<at 1 0 RCIM_Cubic>]}}}, Scale Curve=0.15)
    Scale RGB=&VectorFromFloat(Value=&FloatFromCurve(
      FloatCurve=DataInterface() {propertyValues {Curve {keys [{interpMode=RCIM_Cubic, tangentMode=RCTM_User, time=0, value=1}, {interpMode=RCIM_Cubic, tangentMode=RCTM_User, time=1, value=0}]}}}
      Scale Curve=10
     ))
    ScaleA=true
    ScaleRGB=true
   )
   &FN_Light_Attributes(
    Color=$User.Color_01
    Create Color Attribute=true
    Create Exponent Attribute=true
    Create Radius Attribute=true
    Enable Particle Light Enable=true
    Exponent=3
    Light Enabled=true
    Radius=650
    Scale Color=&FloatFromCurve(
     FloatCurve=DataInterface() {propertyValues {Curve {keys [{interpMode=RCIM_Cubic, tangentMode=RCTM_User, time=0, value=1}, {interpMode=RCIM_Cubic, tangentMode=RCTM_User, time=1, value=0}]}}}
     Scale Curve=100
    )
   )
  ]
  rendererClasses [&NiagaraLightRendererProperties]
  renderers [{rendererClass=&NiagaraLightRendererProperties, rendererIndex=0}]
 }
 OuterGlow {
  bEnabled=false
  emitterUpdateScript [
   &EmitterState()
   &SpawnBurst_Instantaneous(Spawn Count=1, Spawn Time=$User.ExplosionDelay, Use Spawn Probability=false)
  ]
  particleSpawnScript [/Niagara/Modules/Spawn/Initialization/V2/InitializeParticle(
    Color=$User.Color_02
    Lifetime=&Multiply_Float(A=0.3, B=$User.ExplosionLife)
    Mesh Scale Mode=::NewEnumerator3
    Mesh Uniform Scale=&Multiply_Float(A=11.5, B=$User.ImpactSize)
    UsePositionOffset=false
    Write Material Random=true
   ) {module=InitializeParticle001}]
  particleUpdateScript [
   &ParticleState()
   &DynamicMaterialParameters(
    Index 0 Param 1=0.4
    Module.Param0WriteEnabled=true
    Module.Param1WriteEnabled=true
    Module.Param2WriteEnabled=true
    Module.Param3WriteEnabled=true
   ) {module=DynamicMaterialParameters001}
   &ScaleColor(
    Scale Alpha=&FloatFromCurve(
     FloatCurve=DataInterface() {propertyValues {Curve {keys [
         {arriveTangent=45, interpMode=RCIM_Cubic, time=0, value=1}
         {arriveTangent=-1, interpMode=RCIM_Cubic, leaveTangent=-1, time=0.3, value=2}
         <at 1 0 RCIM_Cubic>
        ]}}}
     Scale Curve=0.005
    )
    Scale RGB=<3 3 3>
    ScaleA=true
    ScaleRGB=true
   )
   &ScaleMeshSize(Scale Factor=&VectorFromFloat(Value=&FloatFromCurve(FloatCurve=DataInterface() {propertyValues {Curve {keys [
          {arriveTangent=4.1785, interpMode=RCIM_Cubic, leaveTangent=4.1785, tangentMode=RCTM_User, time=0, value=0.05}
          {arriveTangent=0.9995, interpMode=RCIM_Cubic, leaveTangent=0.9995, tangentMode=RCTM_User, time=0.15, value=0.6825}
          {arriveTangent=-0.0405, interpMode=RCIM_Cubic, leaveTangent=-0.0405, tangentMode=RCTM_User, time=1, value=1}
         ]}}})))
  ]
  rendererClasses [&NiagaraMeshRendererProperties]
  renderers [{rendererClass=&NiagaraMeshRendererProperties, rendererIndex=0}]
 }
 OuterRadialRing {
  emitterUpdateScript [
   &EmitterState()
   &SpawnBurst_Instantaneous(Spawn Count=1, Spawn Time=$User.ExplosionDelay, Use Spawn Probability=false)
  ]
  particleSpawnScript [
   &InitializeParticle(
    Color=$User.Color_02
    Lifetime=&Multiply_Float(A=0.4, B=$User.ExplosionLife)
    Mesh Scale Mode=::NewEnumerator3
    Mesh Uniform Scale=&Multiply_Float(A=11.5, B=$User.ImpactSize)
    UsePositionOffset=false
    Write Material Random=true
   ) {enabled=false, module=InitializeParticle001}
   &InitializeParticle(
    Color=$User.Color_02
    Lifetime=0.35
    Mesh Scale Mode=::NewEnumerator3
    Mesh Uniform Scale=&Multiply_Float(A=$User.ImpactSize, B=8.8)
    Position Offset=<0 0 5>
    UsePositionOffset=true
    Write Material Random=true
   )
   &InitialMeshOrientation(
    Orientation Axis=<0 0 1>
    Rotation=&RandomRangeVector(Minimum=<0 0 0>, Maximum=<0 0 1>)
    Use Orientation Vector=true
    Use Rotation Vector=true
   )
  ]
  particleUpdateScript [
   &ParticleState()
   &DynamicMaterialParameters(
    Index 0 Param 1=0.4
    Module.Param0WriteEnabled=true
    Module.Param1WriteEnabled=true
    Module.Param2WriteEnabled=true
    Module.Param3WriteEnabled=true
   ) {enabled=false, module=DynamicMaterialParameters001}
   &ScaleColor(
    Scale Alpha=&FloatFromCurve(
     FloatCurve=DataInterface() {propertyValues {Curve {keys [
         {arriveTangent=45, interpMode=RCIM_Cubic, time=0, value=1}
         {arriveTangent=-1, interpMode=RCIM_Cubic, leaveTangent=-1, time=0.3, value=2}
         <at 1 0 RCIM_Cubic>
        ]}}}
     Scale Curve=0.15
    )
    Scale RGB=&MakeVectorFromLinearColorRGB(Color=&MultiplyLinearColor(
      Linear Color=&FN_Select_Color_from_Time_Of_Day(Morning=<rgba 0.3 0.25 1.5 1>, Day=<rgba 0.6 0.5 1.5 1>, Evening=<rgba 0.3 0.25 1.5 1>, Night=<rgba 0.8 0.9 1 1>)
      Scale Factor=&MakeVector4_FromFloat(Float=3)
     ))
    ScaleA=true
    ScaleRGB=true
   )
   &ScaleMeshSize(Scale Factor=&VectorFromFloat(Value=&FloatFromCurve(
      FloatCurve=DataInterface() {propertyValues {Curve {keys [
          {arriveTangent=4.1785, interpMode=RCIM_Cubic, leaveTangent=4.1785, tangentMode=RCTM_User, time=0, value=0.05}
          {arriveTangent=0.9995, interpMode=RCIM_Cubic, leaveTangent=0.9995, tangentMode=RCTM_User, time=0.2, value=0.8}
          {arriveTangent=-0.0405, interpMode=RCIM_Cubic, leaveTangent=-0.0405, tangentMode=RCTM_User, time=1, value=1}
         ]}}}
      Scale Curve=0.5
     )))
   &DynamicMaterialParameters(
    Index 0 Param 1=&FloatFromCurve(
     FloatCurve=DataInterface() {propertyValues {Curve {keys [
         {interpMode=RCIM_Cubic, tangentMode=RCTM_User, time=0, value=0}
         {interpMode=RCIM_Cubic, tangentMode=RCTM_User, time=0.1, value=1}
         {interpMode=RCIM_Cubic, tangentMode=RCTM_User, time=1, value=1}
        ]}}}
     Scale Curve=500
    )
    Module.Param0WriteEnabled=true
    Module.Param1WriteEnabled=true
    Module.Param2WriteEnabled=true
    Module.Param3WriteEnabled=false
   )
   &CameraOffset(Camera Offset Amount=50)
  ]
  rendererClasses [&NiagaraMeshRendererProperties]
  renderers [{rendererClass=&NiagaraMeshRendererProperties, rendererIndex=0}]
 }
 OuterRadialSphere {
  emitterUpdateScript [
   &EmitterState()
   &SpawnBurst_Instantaneous(Spawn Count=2, Spawn Time=$User.ExplosionDelay, Use Spawn Probability=false)
  ]
  particleSpawnScript [
   &InitializeParticle(
    Color=$User.Color_02
    Lifetime=&Multiply_Float(A=0.3, B=$User.ExplosionLife)
    Mesh Scale Mode=::NewEnumerator3
    Mesh Uniform Scale=&Multiply_Float(A=11.5, B=$User.ImpactSize)
    UsePositionOffset=false
    Write Material Random=true
   ) {module=InitializeParticle001}
   &InitialMeshOrientation(
    Rotation=&RandomRangeVector(Minimum=<0 0 0>, Maximum=<1 1 1>)
    Use Orientation Vector=true
    Use Rotation Vector=true
   )
  ]
  particleUpdateScript [
   &ParticleState()
   &DynamicMaterialParameters(
    Index 0 Param 1=0.4
    Module.Param0WriteEnabled=true
    Module.Param1WriteEnabled=true
    Module.Param2WriteEnabled=true
    Module.Param3WriteEnabled=true
   ) {module=DynamicMaterialParameters001}
   &ScaleColor(
    Scale Alpha=&FloatFromCurve(
     FloatCurve=DataInterface() {propertyValues {Curve {keys [
         {arriveTangent=45, interpMode=RCIM_Cubic, time=0, value=1}
         {arriveTangent=-1, interpMode=RCIM_Cubic, leaveTangent=-1, time=0.3, value=2}
         <at 1 0 RCIM_Cubic>
        ]}}}
     Scale Curve=0.25
    )
    Scale RGB=<3 3 3>
    ScaleA=true
    ScaleRGB=true
   )
   &ScaleMeshSize(Scale Factor=&VectorFromFloat(Value=&FloatFromCurve(FloatCurve=DataInterface() {propertyValues {Curve {keys [
          {arriveTangent=4.1785, interpMode=RCIM_Cubic, leaveTangent=4.1785, tangentMode=RCTM_User, time=0, value=0.05}
          {arriveTangent=0.9995, interpMode=RCIM_Cubic, leaveTangent=0.9995, tangentMode=RCTM_User, time=0.15, value=0.6825}
          {arriveTangent=-0.0405, interpMode=RCIM_Cubic, leaveTangent=-0.0405, tangentMode=RCTM_User, time=1, value=1}
         ]}}})))
  ]
  rendererClasses [&NiagaraMeshRendererProperties]
  renderers [{rendererClass=&NiagaraMeshRendererProperties, rendererIndex=0}]
 }
 RadialShockwave {
  emitterUpdateScript [
   &EmitterState()
   &SpawnBurst_Instantaneous(Spawn Count=1, Spawn Time=$User.ExplosionDelay, Use Spawn Probability=false)
  ]
  particleSpawnScript [
   &InitializeParticle(
    Color=$User.Color_02
    Lifetime=&Multiply_Float(A=0.35, B=$User.ExplosionLife)
    Position Offset=<0 0 5>
    Sprite Rotation Mode=::NewEnumerator2
    Sprite Size Mode=::NewEnumerator3
    Uniform Sprite Size=&Multiply_Float(A=1750, B=$User.ImpactSize)
    UsePositionOffset=true
    Write Material Random=true
   )
   /NiagaraFN/Sprite/FN_SpriteAlignmentAndFacing(SpriteFacing:Vector3f=&RotateVectorByQuaternion(Vector To Rotate=<0 0 1>, Quaternion=$Engine.Owner.Rotation))
  ]
  particleUpdateScript [
   &ParticleState()
   &ScaleSpriteSize(Scale Factor=&Vector2DFromFloat(Value=&FloatFromCurve(FloatCurve=DataInterface() {propertyValues {Curve {keys [
          {arriveTangent=4.1785, interpMode=RCIM_Cubic, leaveTangent=4.1785, tangentMode=RCTM_User, time=0, value=0}
          {arriveTangent=0.9995, interpMode=RCIM_Cubic, leaveTangent=0.9995, tangentMode=RCTM_User, time=0.15, value=0.8}
          {arriveTangent=-0.0405, interpMode=RCIM_Cubic, leaveTangent=-0.0405, tangentMode=RCTM_User, time=1, value=1.1}
         ]}}})))
   &ScaleColor(
    Scale Alpha=&FloatFromCurve(
     FloatCurve=DataInterface() {propertyValues {Curve {keys [{interpMode=RCIM_Cubic, tangentMode=RCTM_User, time=0, value=1}, {interpMode=RCIM_Cubic, tangentMode=RCTM_User, time=0.5, value=0}]}}}
     Scale Curve=2.65
    )
    Scale RGB=&VectorFromFloat(Value=2)
    ScaleA=true
    ScaleRGB=true
   )
   &FN_Radial_Material(Depth Fade=2, Radial Mask Tile V=0.8)
   /SetVariables(
    Particles.Color_02:LinearColor=&MultiplyLinearColor(Linear Color=$Particles.Color, Scale Factor=<1 1 1 0.5>)
    Particles.Size_02:Vector2f=&Multiply_Vector2(A=$Particles.SpriteSize, B=<0.9 0.9>)
   ) {isSetParameters=true}
  ]
  rendererClasses [&NiagaraSpriteRendererProperties]
  renderers [
   {rendererClass=&NiagaraSpriteRendererProperties}
   {rendererClass=&NiagaraSpriteRendererProperties, rendererIndex=1}
  ]
 }
 Rings {
  bEnabled=false
  emitterUpdateScript [
   &EmitterState()
   &SpawnBurst_Instantaneous(Spawn Count=1, Spawn Time=$User.ExplosionDelay, Use Spawn Probability=false)
  ]
  particleSpawnScript [&InitializeParticle(
    Color=&MultiplyLinearColor(Linear Color=$User.Color_01, Scale Factor=&MakeVector4_FromLinearColor(LinearColor=<rgba 1 1 1 0.1>))
    Lifetime=&Multiply_Float(A=0.5, B=$User.ExplosionLife)
    Mesh Scale Mode=::NewEnumerator3
    Mesh Uniform Scale=&Multiply_Float(A=4, B=$User.ImpactSize)
    UsePositionOffset=false
    Write Material Random=true
   ) {module=InitializeParticle001}]
  particleUpdateScript [
   &ParticleState()
   &ScaleColor(
    Scale Alpha=&FloatFromCurve(
     FloatCurve=DataInterface() {propertyValues {Curve {keys [<at 0 1>, {interpMode=RCIM_Cubic, tangentMode=RCTM_User, time=0.1, value=1}, {interpMode=RCIM_Cubic, tangentMode=RCTM_User, time=1, value=0}]}}}
     Scale Curve=2
    )
    ScaleA=true
    ScaleRGB=true
   )
   &DynamicMaterialParameters(
    Index 0 Param 1=1
    Index 0 Param 2=1
    Index 0 Param 3=1
    Index 0 Param 4=1
    Module.Param0WriteEnabled=true
    Module.Param1WriteEnabled=true
    Module.Param2WriteEnabled=true
    Module.Param3WriteEnabled=true
   )
   &ScaleMeshSize(Scale Factor=&MakeVector(
     X=&FloatFromCurve(FloatCurve=DataInterface() {propertyValues {Curve {keys [
          {arriveTangent=4.1785, interpMode=RCIM_Cubic, leaveTangent=4.1785, tangentMode=RCTM_User, time=0, value=0.05}
          {arriveTangent=0.9995, interpMode=RCIM_Cubic, leaveTangent=0.9995, tangentMode=RCTM_User, time=0.15, value=0.8}
          {arriveTangent=-0.0405, interpMode=RCIM_Cubic, leaveTangent=-0.0405, tangentMode=RCTM_User, time=1, value=1}
         ]}}})
     Y=&FloatFromCurve(FloatCurve=DataInterface() {propertyValues {Curve {keys [
          {arriveTangent=4.1785, interpMode=RCIM_Cubic, leaveTangent=4.1785, tangentMode=RCTM_User, time=0, value=0.05}
          {arriveTangent=0.9995, interpMode=RCIM_Cubic, leaveTangent=0.9995, tangentMode=RCTM_User, time=0.15, value=0.8}
          {arriveTangent=-0.0405, interpMode=RCIM_Cubic, leaveTangent=-0.0405, tangentMode=RCTM_User, time=1, value=1}
         ]}}})
     Z=1
    ))
  ]
  rendererClasses [&NiagaraMeshRendererProperties]
  renderers [{rendererClass=&NiagaraMeshRendererProperties, rendererIndex=0}]
 }
 SmokeShockwave {
  bEnabled=false
  emitterUpdateScript [
   &EmitterState()
   &SpawnBurst_Instantaneous(
    Spawn Count=&RandomRangeInt(Minimum=12, Maximum=16)
    Spawn Time=&Subtract_Float(A=$User.ExplosionDelay, B=0.04)
    Use Spawn Probability=false
   )
  ]
  particleSpawnScript [
   /Niagara/Modules/Spawn/Initialization/InitializeParticle(
    Lifetime=&Multiply_Float(A=&RandomRangeFloat(Minimum=0.6, Maximum=1.2), B=$User.ExplosionLife)
    Module.Write Color=true
    Module.Write Lifetime=true
    Module.Write Mass=true
    Module.Write Material Random=true
    Module.Write Position=false
    Module.Write Scale=false
    Module.Write SpriteRotation=true
    Module.Write SpriteSize=true
    Position=/Niagara/Functions/Localspace/SimulationPosition()
    Sprite Rotation=&UniformRangedFloat(Minimum=0, Maximum=360)
    Sprite Size=&Vector2DFromFloat(Value=&Multiply_Float(A=&RandomRangeFloat(Minimum=13, Maximum=23), B=5))
   )
   &TorusLocation(
    Align Torus With Axis=false
    Large Radius=10
    MaskSpawnGroup=false
    Module.Override Randomness=false
    Surface Position=1
    Torus Mode=::NewEnumerator2
   )
   &SphereLocation(MaskSpawnGroup=false, Override Seed=false, Surface Only=false, UseNonUniformScale=false, UseOffset=false, UseSphereOrientation=false) {enabled=false}
   &AddVelocityFromPoint(
    Module.Use Falloff Distance=false
    Velocity Origin:Vector3f=&SimulationPosition()
    Velocity Strength=&RandomRangeFloat(Minimum=1800, Maximum=2000)
   )
  ]
  particleUpdateScript [
   &ParticleState()
   &Drag(
    Drag=&FloatFromCurve(
     FloatCurve=DataInterface() {propertyValues {Curve {keys [<at 0 2 RCIM_Cubic>,<at 1 8 RCIM_Cubic>]}}}
     Scale Curve=&RandomRangeFloat(Minimum=1, Maximum=1.5)
    )
    Use Linear Drag=true
    Use Rotational Drag=false
   )
   &GravityForce(Gravity=<0 0 -200>)
   &SolveForcesAndVelocity(Clamp Velocity=false, Limit Acceleration=false)
   &ScaleSpriteSize(Scale Factor=&Vector2DFromFloat(Value=&FloatFromCurve(FloatCurve=DataInterface() {propertyValues {Curve {keys [
          {arriveTangent=1.491, interpMode=RCIM_Cubic, leaveTangent=1.491, tangentMode=RCTM_User, time=-0.0046, value=1.1969}
          <at 1.0011 3.4809 RCIM_Cubic>
         ]}}})))
   &DynamicMaterialParameters(
    Index 0 Param 1=&FloatFromCurve(FloatCurve=DataInterface() {propertyValues {Curve {keys [<at 0 0 RCIM_Cubic>,<at 1 0.25>]}}})
    Index 0 Param 4=&FloatFromCurve(FloatCurve=DataInterface() {propertyValues {Curve {keys [<at 0 0 RCIM_Cubic>,<at 1 0.5>]}}}, Scale Curve=0.1)
    Module.Param0WriteEnabled=true
    Module.Param1WriteEnabled=false
    Module.Param2WriteEnabled=false
    Module.Param3WriteEnabled=true
   ) {module=DynamicMatParams}
   &Color(Color=&LerpLinearColors(
     StartColor=$User.Color_02
     EndColor=$User.Color_01
     LerpFactor=&FloatFromCurve(
      FloatCurve=DataInterface() {propertyValues {Curve {keys [<at 0 1>,<at 1 0>]}}}
      CurveIndex=&NormalizeDistanceRange(
       Start Position:NiagaraPosition=$Particles.Position
       End Position:NiagaraPosition=/Niagara/DynamicInputs/Helpers/SimulationPosition()
       Distance:NiagaraFloat=400
       Invert Normalized Range:NiagaraBool=false
      )
     )
    ))
   &ScaleColor(
    Color Value To Scale=$Particles.Color
    Scale Alpha=&FloatFromCurve(
     FloatCurve=DataInterface() {propertyValues {Curve {keys [{interpMode=RCIM_Cubic, tangentMode=RCTM_User, time=0.5, value=1}, {interpMode=RCIM_Cubic, tangentMode=RCTM_User, time=1, value=0}]}}}
     Scale Curve=0.22
    )
    ScaleA=true
    ScaleRGB=false
   )
  ]
  rendererClasses [&NiagaraSpriteRendererProperties]
  renderers [{rendererClass=&NiagaraSpriteRendererProperties}]
 }
 SnapMesh {
  emitterUpdateScript [
   &EmitterState()
   &SpawnBurst_Instantaneous(Spawn Count=1, Spawn Probability=$User.ExplosionDelay, Use Spawn Probability=false)
  ]
  particleSpawnScript [
   &InitializeParticle(
    Lifetime=&RandomRangeFloat(Minimum=0.25, Maximum=0.3, Evaluation Type=::NewEnumerator1)
    Mesh Scale=&VectorFromFloat(Value=&Multiply_Float(A=8, B=$User.ImpactSize))
    Module.Write Color=true
    Module.Write Lifetime=true
    Module.Write Mass=false
    Module.Write Material Random=true
    Module.Write Position=false
    Module.Write Scale=true
    Module.Write SpriteRotation=false
    Module.Write SpriteSize=false
   )
   &InitialMeshOrientation(
    Rotation=&RandomRangeVector(Minimum=<-180 -180 -180>, Maximum=<180 180 180>)
    Use Orientation Vector=true
    Use Rotation Vector=true
   )
  ]
  particleUpdateScript [
   &ParticleState()
   &ScaleColor(
    Scale Alpha=&FloatFromCurve(FloatCurve=DataInterface() {propertyValues {Curve {keys [{tangentMode=RCTM_User, time=0, value=1}, {tangentMode=RCTM_User, time=1, value=0}]}}})
    Scale RGB=&VectorFromFloat(Value=&FloatFromCurve(
      FloatCurve=DataInterface() {propertyValues {Curve {keys [{interpMode=RCIM_Cubic, tangentMode=RCTM_User, time=0, value=1}, {interpMode=RCIM_Cubic, tangentMode=RCTM_User, time=1, value=0}]}}}
      Scale Curve=20
     ))
    ScaleA=true
    ScaleRGB=true
   )
   &DynamicMaterialParameters(
    Index 0 Param 1=&FloatFromCurve(FloatCurve=DataInterface() {propertyValues {Curve {keys [{arriveTangent=0.0592, interpMode=RCIM_Cubic, time=0, value=0}, <at 1 1 RCIM_Cubic>]}}})
    Index 0 Param 2=50
    Index 0 Param 3=1
    Index 0 Param 4=1
    Module.Param0WriteEnabled=true
    Module.Param1WriteEnabled=true
    Module.Param2WriteEnabled=true
    Module.Param3WriteEnabled=true
   )
   &ScaleMeshSize(Scale Factor=&VectorFromFloat(Value=&FloatFromCurve(
      FloatCurve=DataInterface() {propertyValues {Curve {keys [{arriveTangent=2.0369, interpMode=RCIM_Cubic, leaveTangent=2.0369, tangentMode=RCTM_User, time=0, value=0.25}, <at 1 1>]}}}
      Scale Curve=1.8
     )))
  ]
  rendererClasses [&NiagaraMeshRendererProperties]
  renderers [{rendererClass=&NiagaraMeshRendererProperties, rendererIndex=0}]
 }
 Speedlines {
  emitterUpdateScript [
   &EmitterState()
   &SpawnBurst_Instantaneous(Spawn Count=&RandomRangeInt(Minimum=20, Maximum=24), Use Spawn Probability=false)
  ]
  particleSpawnScript [
   &InitializeParticle(
    Color=&LerpLinearColors(StartColor=$User.Color_02, EndColor=$User.Color_01, LerpFactor=&RandomRangeFloat(Minimum=0, Maximum=1))
    Lifetime=&RandomRangeFloat(Minimum=0.3, Maximum=0.35, Evaluation Type=::NewEnumerator1)
    Module.Write Color=true
    Module.Write Lifetime=true
    Module.Write Mass=false
    Module.Write Material Random=true
    Module.Write Position=false
    Module.Write Scale=false
    Module.Write SpriteRotation=false
    Module.Write SpriteSize=true
    Sprite Size=&RandomRangeVector2D(Minimum=<8 18>, Maximum=<12 24>, Evaluation Type=::NewEnumerator1)
   )
   /Niagara/Modules/Spawn/Location/SphereLocation(MaskSpawnGroup=false, Module.Override Randomness=false, Sphere Radius=150, Surface Only=true)
   &AddVelocityFromPoint(Module.Use Falloff Distance=false, Velocity Strength=&RandomRangeFloat(Minimum=600, Maximum=1800))
  ]
  particleUpdateScript [
   &ParticleState()
   &ScaleColor(
    Scale Alpha=&FloatFromCurve(FloatCurve=DataInterface() {propertyValues {Curve {keys [<at 0 1>,<at 1 0>]}}}, Scale Curve=0.5)
    Scale RGB=<3 3 3>
    ScaleA=true
    ScaleRGB=true
   )
   &Drag(Use Linear Drag=true, Use Rotational Drag=false)
   &ScaleSpriteSize(Scale Factor=&Vector2DFromCurve(Vector2Curve=DataInterface() {propertyValues {XCurve {keys [<at 0 1>,<at 1 0>]}, YCurve {keys [<at 0 8>,<at 1 4>]}}}, ScaleCurve=<1 1>))
   &SolveForcesAndVelocity(Clamp Velocity=false, Limit Acceleration=false)
  ]
  rendererClasses [&NiagaraSpriteRendererProperties]
  renderers [{rendererClass=&NiagaraSpriteRendererProperties}]
 }
 SphereRibbon {
  bEnabled=false
  emitterUpdateScript [
   &EmitterState()
   &SpawnBurst_Instantaneous(Spawn Count=&RandomRangeInt(Minimum=20, Maximum=40), Use Spawn Probability=false)
   &SpawnRate(SpawnRate=5, Use Spawn Probability=false) {enabled=false}
  ]
  particleSpawnScript [
   /Niagara/Modules/Spawn/Initialization/V2/InitializeParticle(
    Color=$User.Color_02
    Lifetime=0.35
    Mass Max=2
    Mass Mode=::NewEnumerator3
    Sprite Size Max=<2 8>
    Sprite Size Min=<1.5 6>
    Sprite Size Mode=::NewEnumerator2
    UsePositionOffset=false
    Write Material Random=true
   ) {enabled=false}
   &InitializeRibbon(
    Color=&LerpLinearColors(StartColor=$User.Color_02, EndColor=$User.Color_03, LerpFactor=&RandomRangeFloat(Minimum=0, Maximum=1))
    Lifetime=&Multiply_Float(A=1.2, B=$User.ExplosionLife)
    Module.Write Color=true
    Module.Write Lifetime=true
    Module.Write Mass=true
    Module.Write Material Random=true
    Module.Write Position=false
    Module.Write RibbonTwist=false
    Module.Write RibbonWidth=true
    Ribbon Width=&RandomRangeFloat(Minimum=24, Maximum=30)
   )
   &AddVelocityInCone(
    Cone Angle=48
    Cone Axis=<0.9 0 0.4>
    Module.Use Velocity Falloff On Cone Axis=true
    Velocity Distribution Along Cone Axis=0.75
    Velocity Falloff Away From Cone Axis=0.333
    Velocity Strength=&RandomRangeFloat(Minimum=350, Maximum=750)
   ) {enabled=false}
   &CylinderLocation(
    Cylinder Height=0
    Cylinder Radius=&Multiply_Float(A=$User.ImpactSize, B=50)
    MaskSpawnGroup=false
    Module.Enable Lathe Profile=false
    Module.Override Randomness=false
    Module.Surface Only=true
    Surface Only Band Thickness=0
    Use Endcaps In Surface Only Mode=false
   ) {enabled=false}
   /Niagara/Modules/Spawn/Location/V2/SphereLocation(MaskSpawnGroup=false, Override Seed=false, Surface Only=false, UseNonUniformScale=false, UseOffset=false, UseSphereOrientation=false) {module=SphereLocation001}
   &InheritVelocity(Module.Limit Speed=true, Velocity Limit=250, Velocity Scale=<-1 -1 -1>) {enabled=false}
   &SphereLocation(MaskSpawnGroup=false, Override Seed=false, Surface Only=false, UseNonUniformScale=false, UseOffset=false, UseSphereOrientation=false) {enabled=false}
   &FN_SetSpawnGroup(Number of Spawn Groups=1, Set Ribbon ID and Ribbon Link Order=true)
   &AddVelocityFromPoint(
    Module.Use Falloff Distance=false
    Velocity Strength=&Multiply_Float(A=&RandomRangeFloat(Minimum=1300, Maximum=1400), B=0.75)
   )
  ]
  particleUpdateScript [
   &ParticleState()
   &GravityForce(Gravity=<0 0 -200>) {enabled=false}
   &Drag(
    Drag=&FloatFromCurve(FloatCurve=DataInterface() {propertyValues {Curve {keys [{interpMode=RCIM_Cubic, tangentMode=RCTM_User, time=0, value=0}, {interpMode=RCIM_Cubic, tangentMode=RCTM_User, time=0.125, value=2.5}]}}})
    Use Linear Drag=true
    Use Rotational Drag=false
   )
   &CurlNoiseForce(
    Noise Frequency=20
    Noise Strength=&Multiply_Float(
     A=3500
     B=&FloatFromCurve(FloatCurve=DataInterface() {propertyValues {Curve {keys [<at 0 1>,<at 0.5 0>]}}})
    )
    Pan Noise=false
   )
   &SolveForcesAndVelocity(Clamp Velocity=false, Limit Acceleration=false)
   &ScaleSpriteSizeBySpeed(Max Scale Factor=<1.2 5>, Velocity Threshold=1500) {enabled=false}
   &ScaleSpriteSize(Scale Factor=&Vector2DFromFloat(Value=&FloatFromCurve(FloatCurve=DataInterface() {propertyValues {Curve {keys [
          {arriveTangent=9.0725, interpMode=RCIM_Cubic, leaveTangent=9.0725, tangentMode=RCTM_User, time=0, value=0}
          {arriveTangent=-2.5249, interpMode=RCIM_Cubic, leaveTangent=-2.5249, tangentMode=RCTM_User, time=0.1408, value=0.9847}
          {arriveTangent=-0.0405, interpMode=RCIM_Cubic, leaveTangent=-0.0405, tangentMode=RCTM_User, time=1, value=0}
         ]}}}))) {enabled=false}
   &ScaleColor(
    Scale Alpha=&FloatFromCurve(
     FloatCurve=DataInterface() {propertyValues {Curve {keys [
         {interpMode=RCIM_Cubic, tangentMode=RCTM_User, time=0, value=0}
         {interpMode=RCIM_Cubic, tangentMode=RCTM_User, time=0.25, value=1}
         {arriveTangent=-1.3333, interpMode=RCIM_Cubic, leaveTangent=-1.3333, time=0.4, value=0.25}
         {tangentMode=RCTM_User, time=1, value=0}
        ]}}}
     Scale Curve=&RandomRangeFloat(Minimum=0.5, Maximum=1.5)
    )
    ScaleA=true
    ScaleRGB=true
   )
   &ScaleRibbonWidth(Ribbon Width Scale=&FloatFromCurve(
     FloatCurve=DataInterface() {propertyValues {Curve {keys [{tangentMode=RCTM_User, time=0.9, value=0}, {tangentMode=RCTM_User, time=1, value=3}]}}}
     CurveIndex=&NormalizeDistanceRange(
      Start Position:NiagaraPosition=$Particles.Position
      End Position:NiagaraPosition=$Engine.Owner.Position
      Distance:NiagaraFloat=1
      Invert Normalized Range:NiagaraBool=false
     )
    )) {enabled=false}
   &ScaleColor(Color Value To Scale=$Particles.Color, Scale Alpha=220, ScaleA=true, ScaleRGB=true) {enabled=false, module=ScaleColor001}
  ]
  rendererClasses [&NiagaraSpriteRendererProperties, &NiagaraRibbonRendererProperties]
  renderers [
   {rendererClass=&NiagaraSpriteRendererProperties}
   {rendererClass=&NiagaraRibbonRendererProperties, rendererIndex=1}
  ]
 }
 SphereRibbon_02 {
  bEnabled=false
  emitterUpdateScript [
   &EmitterState()
   &SpawnBurst_Instantaneous(Spawn Count=&RandomRangeInt(Minimum=30, Maximum=50), Use Spawn Probability=false)
   &SpawnRate(SpawnRate=5, Use Spawn Probability=false) {enabled=false}
  ]
  particleSpawnScript [
   &InitializeParticle(
    Color=$User.Color_02
    Lifetime=0.35
    Mass Max=2
    Mass Mode=::NewEnumerator3
    Sprite Size Max=<2 8>
    Sprite Size Min=<1.5 6>
    Sprite Size Mode=::NewEnumerator2
    UsePositionOffset=false
    Write Material Random=true
   ) {enabled=false}
   &InitializeRibbon(
    Color=&LerpLinearColors(StartColor=$User.Color_02, EndColor=$User.Color_03, LerpFactor=&RandomRangeFloat(Minimum=0, Maximum=1))
    Lifetime=&Multiply_Float(A=1.2, B=$User.ExplosionLife)
    Module.Write Color=true
    Module.Write Lifetime=true
    Module.Write Mass=true
    Module.Write Material Random=true
    Module.Write Position=false
    Module.Write RibbonTwist=false
    Module.Write RibbonWidth=true
    Ribbon Width=&RandomRangeFloat(Minimum=24, Maximum=30)
   )
   &AddVelocityInCone(
    Cone Angle=48
    Cone Axis=<0.9 0 0.4>
    Module.Use Velocity Falloff On Cone Axis=true
    Velocity Distribution Along Cone Axis=0.75
    Velocity Falloff Away From Cone Axis=0.333
    Velocity Strength=&RandomRangeFloat(Minimum=350, Maximum=750)
   ) {enabled=false}
   &CylinderLocation(
    Cylinder Height=0
    Cylinder Radius=&Multiply_Float(A=$User.ImpactSize, B=50)
    MaskSpawnGroup=false
    Module.Enable Lathe Profile=false
    Module.Override Randomness=false
    Module.Surface Only=true
    Surface Only Band Thickness=0
    Use Endcaps In Surface Only Mode=false
   ) {enabled=false}
   &SphereLocation(MaskSpawnGroup=false, Override Seed=false, Surface Only=false, UseNonUniformScale=false, UseOffset=false, UseSphereOrientation=false) {module=SphereLocation001}
   &InheritVelocity(Module.Limit Speed=true, Velocity Limit=250, Velocity Scale=<-1 -1 -1>) {enabled=false}
   &SphereLocation(MaskSpawnGroup=false, Override Seed=false, Surface Only=false, UseNonUniformScale=false, UseOffset=false, UseSphereOrientation=false) {enabled=false}
   &FN_SetSpawnGroup(Number of Spawn Groups=1, Set Ribbon ID and Ribbon Link Order=true)
   &AddVelocityFromPoint(
    Module.Use Falloff Distance=false
    Velocity Strength=&Multiply_Float(A=&RandomRangeFloat(Minimum=1400, Maximum=1500), B=0.75)
   )
  ]
  particleUpdateScript [
   &ParticleState()
   &GravityForce(Gravity=<0 0 -200>) {enabled=false}
   &Drag(
    Drag=&FloatFromCurve(FloatCurve=DataInterface() {propertyValues {Curve {keys [{interpMode=RCIM_Cubic, tangentMode=RCTM_User, time=0, value=0}, {interpMode=RCIM_Cubic, tangentMode=RCTM_User, time=0.125, value=2.5}]}}})
    Use Linear Drag=true
    Use Rotational Drag=false
   )
   &CurlNoiseForce(
    Noise Frequency=20
    Noise Strength=&Multiply_Float(
     A=3500
     B=&FloatFromCurve(FloatCurve=DataInterface() {propertyValues {Curve {keys [<at 0 1>,<at 0.5 0>]}}})
    )
    Pan Noise=false
   )
   &SolveForcesAndVelocity(Clamp Velocity=false, Limit Acceleration=false)
   &ScaleSpriteSizeBySpeed(Max Scale Factor=<1.2 5>, Velocity Threshold=1500) {enabled=false}
   &ScaleSpriteSize(Scale Factor=&Vector2DFromFloat(Value=&FloatFromCurve(FloatCurve=DataInterface() {propertyValues {Curve {keys [
          {arriveTangent=9.0725, interpMode=RCIM_Cubic, leaveTangent=9.0725, tangentMode=RCTM_User, time=0, value=0}
          {arriveTangent=-2.5249, interpMode=RCIM_Cubic, leaveTangent=-2.5249, tangentMode=RCTM_User, time=0.1408, value=0.9847}
          {arriveTangent=-0.0405, interpMode=RCIM_Cubic, leaveTangent=-0.0405, tangentMode=RCTM_User, time=1, value=0}
         ]}}}))) {enabled=false}
   &ScaleColor(
    Scale Alpha=&FloatFromCurve(
     FloatCurve=DataInterface() {propertyValues {Curve {keys [
         {interpMode=RCIM_Cubic, tangentMode=RCTM_User, time=0, value=0}
         {interpMode=RCIM_Cubic, tangentMode=RCTM_User, time=0.15, value=1}
         {arriveTangent=-2.0861, interpMode=RCIM_Cubic, leaveTangent=-2.0861, tangentMode=RCTM_User, time=0.25, value=0.2}
         {tangentMode=RCTM_User, time=0.6, value=0}
        ]}}}
     Scale Curve=&RandomRangeFloat(Minimum=0.5, Maximum=1.5)
    )
    ScaleA=true
    ScaleRGB=true
   )
   &ScaleRibbonWidth(Ribbon Width Scale=&FloatFromCurve(
     FloatCurve=DataInterface() {propertyValues {Curve {keys [{tangentMode=RCTM_User, time=0.9, value=0}, {tangentMode=RCTM_User, time=1, value=3}]}}}
     CurveIndex=&NormalizeDistanceRange(
      Start Position:NiagaraPosition=$Particles.Position
      End Position:NiagaraPosition=$Engine.Owner.Position
      Distance:NiagaraFloat=1
      Invert Normalized Range:NiagaraBool=false
     )
    )) {enabled=false}
   &ScaleColor(Color Value To Scale=$Particles.Color, Scale Alpha=220, ScaleA=true, ScaleRGB=true) {enabled=false, module=ScaleColor001}
  ]
  rendererClasses [&NiagaraSpriteRendererProperties, &NiagaraRibbonRendererProperties]
  renderers [
   {rendererClass=&NiagaraSpriteRendererProperties}
   {rendererClass=&NiagaraRibbonRendererProperties, rendererIndex=1}
  ]
 }
}
summary {
 systemName=NS_KnockbackWeapon_Impact
 emitters [
  {emitterName=InnerRadialSphere, rendererClasses [&NiagaraMeshRendererProperties]}
  {emitterName=OuterRadialSphere, rendererClasses [&NiagaraMeshRendererProperties]}
  {emitterName=RadialShockwave, rendererClasses [&NiagaraSpriteRendererProperties]}
  {bEnabled=false, emitterName=SmokeShockwave, rendererClasses [&NiagaraSpriteRendererProperties]}
  {emitterName=SnapMesh, rendererClasses [&NiagaraMeshRendererProperties]}
  {emitterName=Speedlines, rendererClasses [&NiagaraSpriteRendererProperties]}
  {
   bEnabled=false
   emitterName=GroundRibbon
   rendererClasses [&NiagaraSpriteRendererProperties, &NiagaraRibbonRendererProperties]
  }
  {emitterName=AirGlow, rendererClasses [&NiagaraSpriteRendererProperties]}
  {emitterName=Light, rendererClasses [&NiagaraLightRendererProperties]}
  {
   bEnabled=false
   emitterName=SphereRibbon
   rendererClasses [&NiagaraSpriteRendererProperties, &NiagaraRibbonRendererProperties]
  }
  {emitterName=InnerNoiseSphere, rendererClasses [&NiagaraMeshRendererProperties]}
  {bEnabled=false, emitterName=InnerGlowSphere, rendererClasses [&NiagaraMeshRendererProperties]}
  {bEnabled=false, emitterName=OuterGlow, rendererClasses [&NiagaraMeshRendererProperties]}
  {bEnabled=false, emitterName=FN_Mesh, rendererClasses [&NiagaraMeshRendererProperties]}
  {emitterName=Distortion, rendererClasses [&NiagaraMeshRendererProperties]}
  {bEnabled=false, emitterName=Rings, rendererClasses [&NiagaraMeshRendererProperties]}
  {bEnabled=false, emitterName=ImpactRingsPos, rendererClasses [&NiagaraMeshRendererProperties]}
  {bEnabled=false, emitterName=ImpactRingsNeg, rendererClasses [&NiagaraMeshRendererProperties]}
  {
   bEnabled=false
   emitterName=SphereRibbon_02
   rendererClasses [&NiagaraSpriteRendererProperties, &NiagaraRibbonRendererProperties]
  }
  {bEnabled=false, emitterName=AirGlow001, rendererClasses [&NiagaraSpriteRendererProperties]}
  {bEnabled=false, emitterName=FN_RadialSprite, rendererClasses [&NiagaraSpriteRendererProperties]}
  {emitterName=OuterRadialRing, rendererClasses [&NiagaraMeshRendererProperties]}
  {emitterName=FN_Core_Smoke_Sprite, rendererClasses [&NiagaraSpriteRendererProperties]}
 ]
 userVariables [
  User.ADS {type=NiagaraFloat, defaultValue=0, description=""}
  User.Color_01 {type=LinearColor, description="", defaultValue {struct=LinearColor, value=<rgba 0 1 0.8115 1>}}
  User.Color_02 {type=LinearColor, description="", defaultValue {struct=LinearColor, value=<rgba 1 0.2104 0.9156 1>}}
  User.Color_03 {type=LinearColor, description="", defaultValue {struct=LinearColor, value=<rgba 0.0664 0.2855 0.625 1>}}
  User.ExplosionDelay {type=NiagaraFloat, defaultValue=0.05, description=""}
  User.ExplosionLife {type=NiagaraFloat, defaultValue=0.85, description=""}
  User.ImpactSize {type=NiagaraFloat, defaultValue=1.1, description=""}
  User.NewLinearColor {type=LinearColor, description="", defaultValue {struct=LinearColor, value=<rgba 1 0.9911 1 1>}}
  User.RingSize {type=NiagaraFloat, defaultValue=0, description=""}
  User.SkeletalMesh {
   type=NiagaraDataInterfaceSkeletalMesh
   defaultValue {
    struct=NiagaraExt_VariableValue_DataInterface
    value {dataInterface=&self:NiagaraDataInterfaceSkeletalMesh_0}
   }
  }
 ]
}
